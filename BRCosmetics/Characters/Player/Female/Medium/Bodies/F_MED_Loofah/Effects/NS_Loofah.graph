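# Niagara cosmetic effect for the F_MED_Loofah female player body, adding color-cycling glows to the arms and feet.
# Three sprite emitters (Feet, LeftArm, RightArm) spawn continuously and follow indices picked from per-limb arrays via SkeletalMeshLocation in particle update.
# Two light emitters repeatedly burst at each foot, shifting color and fading over each particle's life.
# All five sample the system's skeletal mesh; a User.bisFootHidden bool is exposed.
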NiagaraSystem /BRCosmetics/Characters/Player/Female/Medium/Bodies/F_MED_Loofah/Effects/NS_Loofah
dependencies {
 usedDataInterfaces [
  &NiagaraDataInterfaceVectorCurve
  &NiagaraDataInterfaceCurve
  &NiagaraDataInterfaceArrayInt32
  &NiagaraDataInterfaceSkeletalMesh
 ]
 usedDynamicInputs [&InvertBool]
 usedModules [
  &self:SystemSpawnScript.SystemScriptSource.SystemScriptGraph.NiagaraNodeAssignment_0.SetVariables_#1
  &SystemState
  &self:SystemSpawnScript.SystemScriptSource.SystemScriptGraph.NiagaraNodeAssignment_2.SetVariables_#2
 ]
}
emitters {
 Feet {
  emitterUpdateScript [
   &EmitterState(Life Cycle Mode=::NewEnumerator1, Loop Duration=1, UseLoopDelay=true)
   &SpawnRate(SpawnRate=5, Use Spawn Probability=false)
  ]
  particleSpawnScript [
   &InitializeParticle(
    Lifetime=&RandomRangeFloat(Minimum=0.9, Maximum=1.4)
    Sprite Size=<18 18>
    Sprite Size Mode=::NewEnumerator4
    UsePositionOffset=false
    Write Material Random=true
   )
   &Color(Color=&MakeLinearColorFromVectorAndFloat("Vector (RGB)"=<1.35 1.35 1.35>, "Float (Alpha)"=&ClampFloat(Float=1)))
  ]
  particleUpdateScript [
   &ParticleState()
   &ScaleColor(
    Scale Alpha=&FloatFromCurve(FloatCurve=DataInterface() {propertyValues {Curve {keys [<at 0 0>,<at 0.5 0.75>,<at 1 0>]}}})
    Scale RGB=&VectorFromCurve(VectorCurve=DataInterface() {propertyValues {
       XCurve {keys [<at 0 0>,<at 0.2 1>,<at 0.5 1>,<at 0.72 0.2062>,<at 1 0>]}
       YCurve {keys [<at 0 0.2767>,<at 0.2 0>,<at 0.5 0.8233>,<at 0.72 1>,<at 1 1>]}
       ZCurve {keys [<at 0 1>,<at 0.2 0.022>,<at 0.5 0>,<at 0.72 0>,<at 1 0.8882>]}
      }})
    ScaleA=true
    ScaleRGB=true
   )
   &CameraOffset(Camera Offset Amount=22)
   &SkeletalMeshLocation(
    Bone / Socket Index=&SelectIntFromArray(
     Array Sampling Mode=::NewEnumerator1
     Direct Array Index=?
     Int Selection Array=DataInterface() {propertyValues {IntData [4,5]}}
     Random Seed=0
    )
    Bone Sampling Mode=::NewEnumerator4
    MaskSpawnGroup=false
    Skeletal Mesh=$System.Skeletal Mesh
   )
  ]
  rendererClasses [&NiagaraSpriteRendererProperties]
  renderers [{rendererClass=&NiagaraSpriteRendererProperties}]
 }
 LeftArm {
  emitterUpdateScript [
   &EmitterState(Life Cycle Mode=::NewEnumerator1, Loop Duration=1, UseLoopDelay=true)
   &SpawnRate(SpawnRate=5, Use Spawn Probability=false)
  ]
  particleSpawnScript [
   &InitializeParticle(
    Lifetime=&RandomRangeFloat(Minimum=0.9, Maximum=1.4)
    Sprite Size=<14 14>
    Sprite Size Mode=::NewEnumerator4
    UsePositionOffset=false
    Write Material Random=true
   )
   &Color(Color=&MakeLinearColorFromVectorAndFloat("Vector (RGB)"=<1.35 1.35 1.35>, "Float (Alpha)"=&ClampFloat(Float=1)))
  ]
  particleUpdateScript [
   &ParticleState()
   &ScaleColor(
    Scale Alpha=&FloatFromCurve(FloatCurve=DataInterface() {propertyValues {Curve {keys [<at 0 0>,<at 0.5 1>,<at 1 0>]}}})
    Scale RGB=&VectorFromCurve(VectorCurve=DataInterface() {propertyValues {
       XCurve {keys [<at 0 0>,<at 0.2 1>,<at 0.5 1>,<at 0.72 0.2062>,<at 1 0>]}
       YCurve {keys [<at 0 0.2767>,<at 0.2 0>,<at 0.5 0.8233>,<at 0.72 1>,<at 1 1>]}
       ZCurve {keys [<at 0 1>,<at 0.2 0.022>,<at 0.5 0>,<at 0.72 0>,<at 1 0.8882>]}
      }})
    ScaleA=true
    ScaleRGB=true
   )
   &CameraOffset(Camera Offset Amount=7)
   &SkeletalMeshLocation(
    Bone / Socket Index=&SelectIntFromArray(
     Array Sampling Mode=::NewEnumerator1
     Direct Array Index=?
     Int Selection Array=DataInterface() {propertyValues {IntData [1,3]}}
     Random Seed=0
    )
    Bone Sampling Mode=::NewEnumerator4
    MaskSpawnGroup=false
    Skeletal Mesh=$System.Skeletal Mesh
   )
  ]
  rendererClasses [&NiagaraSpriteRendererProperties]
  renderers [{rendererClass=&NiagaraSpriteRendererProperties}]
 }
 RightArm {
  emitterUpdateScript [
   &EmitterState(Life Cycle Mode=::NewEnumerator1, Loop Duration=1, UseLoopDelay=true)
   &SpawnRate(SpawnRate=5, Use Spawn Probability=false)
  ]
  particleSpawnScript [
   &InitializeParticle(
    Lifetime=&RandomRangeFloat(Minimum=0.9, Maximum=1.4)
    Sprite Size=<14 14>
    Sprite Size Mode=::NewEnumerator4
    UsePositionOffset=false
    Write Material Random=true
   )
   &Color(Color=&MakeLinearColorFromVectorAndFloat("Vector (RGB)"=<1.35 1.35 1.35>, "Float (Alpha)"=&ClampFloat(Float=1)))
  ]
  particleUpdateScript [
   &ParticleState()
   &ScaleColor(
    Scale Alpha=&FloatFromCurve(FloatCurve=DataInterface() {propertyValues {Curve {keys [<at 0 0>,<at 0.5 1>,<at 1 0>]}}})
    Scale RGB=&VectorFromCurve(VectorCurve=DataInterface() {propertyValues {
       XCurve {keys [<at 0 0>,<at 0.2 1>,<at 0.5 1>,<at 0.72 0.2062>,<at 1 0>]}
       YCurve {keys [<at 0 0.2767>,<at 0.2 0>,<at 0.5 0.8233>,<at 0.72 1>,<at 1 1>]}
       ZCurve {keys [<at 0 1>,<at 0.2 0.022>,<at 0.5 0>,<at 0.72 0>,<at 1 0.8882>]}
      }})
    ScaleA=true
    ScaleRGB=true
   )
   &CameraOffset(Camera Offset Amount=7)
   &SkeletalMeshLocation(
    Bone / Socket Index=&SelectIntFromArray(
     Array Sampling Mode=::NewEnumerator1
     Direct Array Index=?
     Int Selection Array=DataInterface() {propertyValues {IntData [0,2]}}
     Random Seed=0
    )
    Bone Sampling Mode=::NewEnumerator4
    MaskSpawnGroup=false
    Skeletal Mesh=$System.Skeletal Mesh
   )
  ]
  rendererClasses [&NiagaraSpriteRendererProperties]
  renderers [{rendererClass=&NiagaraSpriteRendererProperties}]
 }
 foot_l_light {
  emitterUpdateScript [
   &EmitterState(Life Cycle Mode=::NewEnumerator1, Loop Duration=1.75, Recalculate Duration Each Loop=true, UseLoopDelay=true)
   &SpawnBurst_Instantaneous(Use Spawn Probability=false, UseLoopCountLimit=false)
  ]
  particleSpawnScript [
   &InitializeParticle(Lifetime=2, Sprite Size=<1 1>, Sprite Size Mode=::NewEnumerator4, UsePositionOffset=false, Write Material Random=true)
   &Color(Color=&MakeLinearColorFromVectorAndFloat(
     "Vector (RGB)"=&RandomRangeVector(Minimum=<3 3 3>, Maximum=<5 5 5>)
     "Float (Alpha)"=&ClampFloat(Float=1)
    ))
  ]
  particleUpdateScript [
   &ParticleState()
   &ScaleColor(
    Scale Alpha=&FloatFromCurve(FloatCurve=DataInterface() {propertyValues {Curve {keys [<at 0 0 RCIM_Cubic>, <at 0.25 1>, {arriveTangent=-1.3333, interpMode=RCIM_Cubic, leaveTangent=-1.3333, time=0.75, value=1}, <at 1 0>]}}})
    Scale RGB=&VectorFromCurve(VectorCurve=DataInterface() {propertyValues {
       XCurve {keys [<at 0 0>,<at 0.2 0>,<at 0.4 4>,<at 0.6 4>,<at 0.8 4>]}
       YCurve {keys [<at 0 4>,<at 0.2 1.1048>,<at 0.4 0>,<at 0.6 0.4651>,<at 0.8 3.7752>]}
       ZCurve {keys [<at 0 0.3338>,<at 0.2 4>,<at 0.4 3.9962>,<at 0.6 0>,<at 0.8 0>]}
      }})
    ScaleA=true
    ScaleRGB=true
   )
   &CascadeConversion_LightProperties(LightBrightnessScale=50, LightColorScale=<1 1 1>, LightRadiusScale=20)
   &SkeletalMeshLocation(Bone / Socket Index=4, Bone Sampling Mode=::NewEnumerator4, MaskSpawnGroup=false, Skeletal Mesh=$System.Skeletal Mesh)
  ]
  rendererClasses [&NiagaraLightRendererProperties]
  renderers [{rendererClass=&NiagaraLightRendererProperties, rendererIndex=0}]
 }
 foot_r_Light {
  emitterUpdateScript [
   &EmitterState(Life Cycle Mode=::NewEnumerator1, Loop Duration=1.75, Recalculate Duration Each Loop=true, UseLoopDelay=true)
   &SpawnBurst_Instantaneous(Use Spawn Probability=false, UseLoopCountLimit=false)
  ]
  particleSpawnScript [
   &InitializeParticle(Lifetime=2, Sprite Size=<1 1>, Sprite Size Mode=::NewEnumerator4, UsePositionOffset=false, Write Material Random=true)
   &Color(Color=&MakeLinearColorFromVectorAndFloat(
     "Vector (RGB)"=&RandomRangeVector(Minimum=<1 1 1>, Maximum=<5 5 5>)
     "Float (Alpha)"=&ClampFloat(Float=1)
    ))
  ]
  particleUpdateScript [
   &ParticleState()
   &ScaleColor(
    Scale Alpha=&FloatFromCurve(FloatCurve=DataInterface() {propertyValues {Curve {keys [<at 0 0 RCIM_Cubic>, <at 0.25 1>, {arriveTangent=-1.3333, interpMode=RCIM_Cubic, leaveTangent=-1.3333, time=0.75, value=1}, <at 1 0>]}}})
    Scale RGB=&VectorFromCurve(VectorCurve=DataInterface() {propertyValues {
       XCurve {keys [<at 0 2>,<at 0.2 2>,<at 0.4 0>,<at 0.6 0>,<at 0.8 0.88>]}
       YCurve {keys [<at 0 0>,<at 0.2 1.7785>,<at 0.4 2>,<at 0.6 1.8007>,<at 0.8 0>]}
       ZCurve {keys [<at 0 0.3355>,<at 0.2 0>,<at 0.4 0.2592>,<at 0.6 2>,<at 0.8 2>]}
      }})
    ScaleA=true
    ScaleRGB=true
   )
   &CascadeConversion_LightProperties(LightBrightnessScale=60, LightColorScale=<1 1 1>, LightRadiusScale=20)
   &SkeletalMeshLocation(Bone / Socket Index=5, Bone Sampling Mode=::NewEnumerator4, MaskSpawnGroup=false, Skeletal Mesh=$System.Skeletal Mesh)
  ]
  rendererClasses [&NiagaraLightRendererProperties]
  renderers [{rendererClass=&NiagaraLightRendererProperties, rendererIndex=0}]
 }
}
summary {
 systemName=NS_Loofah
 emitters [
  {emitterName=RightArm, rendererClasses [&NiagaraSpriteRendererProperties]}
  {emitterName=LeftArm, rendererClasses [&NiagaraSpriteRendererProperties]}
  {emitterName=Feet, rendererClasses [&NiagaraSpriteRendererProperties]}
  {emitterName=foot_r_Light, rendererClasses [&NiagaraLightRendererProperties]}
  {emitterName=foot_l_light, rendererClasses [&NiagaraLightRendererProperties]}
 ]
 userVariables [User.bisFootHidden {type=NiagaraBool}]
}
