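# Material function for the F_MED_Loofah character body that layers an animated lava lamp effect onto input material attributes through a channel-selectable mask.
# A scrolling gradient over bounds-normalized local position, warped by panning distorted noise, indexes a curve atlas for color.
# Swirl normals drive view-dependent base color and emissive and blend into the input normal.
# A parameter lerps distortion and normals between UV and pre-skinned world-aligned projection, keeping the latter stable under animation.
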MaterialFunction /BRCosmetics/Characters/Player/Female/Medium/Bodies/F_MED_Loofah/Materials/MF_LavaLamp
graph {nodes [
  n1 = CameraVectorWS()
  n2 = GetMaterialAttributes(None=FunctionInput() {inputName=Material, inputType=FunctionInput_MaterialAttributes}) {attributeGetTypes [#1, #2]}
  n3 = DotProduct(A=@n1, B=Transform(None=@n2.Normal))
  n4 = OneMinus(None=DotProduct(A=@n1, B=VertexNormalWS()))
  Add(A=@n3, B=@n4)
  n5 = Time()
  n6 = VectorParameter() {parameterName=LL_WP_PanSpeedAndDirection, defaultValue=<rgba 0 0.1 0.1 1>}
  n7 = Divide(A=VertexInterpolator(VS=PreSkinnedPosition()).PS, B=ScalarParameter() {defaultValue=256, parameterName=LL_WP_GlobalScale})
  n8 = TextureObjectParameter() {parameterName=LL_WP_DistortionNoise, samplerType=SAMPLERTYPE_Masks, texture=/Game/Global/Textures/Default/Noise/T_WrapEdgeWear_Noise}
  n9 = Multiply(A=Time(), B=VectorParameter() {parameterName=LL_WP_DistortionPanSpeedAndDirection, defaultValue=<rgba 0 -0.05 -0.05 1>}.RGB)
  n10 = VertexInterpolator(VS=PreSkinnedNormal())
  n11 = Multiply(A=MaterialFunctionCall(
    TextureObject=@n8
    TextureSize=VectorParameter() {parameterName=LL_WP_DistortionNoiseTiling, defaultValue=<rgba 1 1 1 1>}.RGB
    WorldPosition=Add(A=@n7, B=@n9)
    World Space Normal=@n10.PS
   ) {materialFunction=/Game/Weapons/WeaponSkins/Wraps/MaterialFunctions/WorldAlignedTexture}.XYZ Texture) {constB=2}
  n12 = Add(
   A=Add(A=Multiply(A=@n5, B=@n6.RGB), B=@n7)
   B=Multiply(A=Subtract(A=@n11), B=ScalarParameter() {defaultValue=0.02, parameterName=LL_WP_DistortionStrength})
  )
  Add(A=@n12, B=Multiply(A=ComponentMask() {b=true}, B=ScalarParameter() {defaultValue=12, parameterName=LL_SeamDistortionStrength}))
  n13 = Power(Base=DotProduct(A=CameraVectorWS(), B=VertexNormalWS()))
  n14 = VectorParameter() {parameterName=LL_NoiseTiling, defaultValue=<rgba 0.5 0.5 0.5 1>}
  n15 = ComponentMask(None=VectorParameter() {parameterName=LL_GlobalTiling, defaultValue=<rgba 5 5 1 1>}.RGB) {g=true, r=true}
  n16 = Multiply(A=TextureCoordinate(), B=Multiply(A=ComponentMask(None=@n14.RGB) {g=true, r=true}, B=@n15))
  n17 = ComponentMask(None=VectorParameter() {parameterName=LL_PanSpeedAndDirection, defaultValue=<rgba 0 0.1 0 1>}.RGB) {g=true, r=true}
  n18 = TextureCoordinate()
  n19 = Multiply(A=ComponentMask(None=VectorParameter() {parameterName=LL_DistortionNoiseTiling, defaultValue=<rgba 1 1 1 1>}.RGB) {g=true, r=true}, B=@n15)
  n20 = VectorParameter() {parameterName=LL_DistortionPanSpeedAndDirection, defaultValue=<rgba 0 -0.05 0 1>}
  n21 = TextureSampleParameter2D(UVs=Panner(Coordinate=Multiply(A=@n18, B=@n19), Speed=ComponentMask(None=@n20.RGB) {g=true, r=true})) {parameterName=LL_DistortionNoise_01, samplerType=SAMPLERTYPE_Masks, texture=&T_WrapEdgeWear_Noise}
  n22 = Multiply(A=Subtract(A=Multiply(A=@n21.RGB) {constB=2}), B=ScalarParameter() {defaultValue=0.02, parameterName=LL_DistortionStrength})
  n23 = ComponentMask(None=@n2.Normal) {b=true}
  n24 = ScalarParameter() {defaultValue=0.2, parameterName=LL_SeamDistortion}
  n25 = Add(
   A=Add(A=Panner(Coordinate=@n16, Speed=@n17), B=ComponentMask(None=@n22) {g=true, r=true})
   B=Multiply(A=Multiply(A=@n23, B=@n24), B=VertexColor().R)
  )
  n26 = TextureSampleParameter2D(UVs=@n25) {parameterName=LL_Noise_01, samplerType=SAMPLERTYPE_Masks, texture=&T_WrapEdgeWear_Noise}
  n27 = Subtract(A=Clamp(None=Multiply(A=@n13) {constB=2}), B=Multiply(A=@n26.R, B=ScalarParameter() {defaultValue=1, parameterName=LL_SubtractionStrength}))
  n28 = Round(None=@n27)
  Add(A=@n28, B=Power(Base=@n27))
  n29 = PreSkinnedPosition()
  n30 = Divide(
   A=FunctionInput(Preview=VertexInterpolator(VS=@n29).PS) {bUsePreviewValueAsDefault=true, inputName=Local Position}
   B=PreSkinnedLocalBounds().Max
  )
  n31 = Multiply(A=@n26.RGB) {constB=2}
  n32 = VectorParameter() {parameterName=LL_WP_NoiseTiling, defaultValue=<rgba 0.5 0.5 0.5 1>}
  n33 = Multiply(A=MaterialFunctionCall(
    TextureObject=TextureObjectParameter() {parameterName=LL_WP_Noise_01, samplerType=SAMPLERTYPE_Masks, texture=&T_WrapEdgeWear_Noise}
    TextureSize=@n32.RGB
    WorldPosition=@n12
    World Space Normal=@n10.PS
   ) {materialFunction=&WorldAlignedTexture}.XYZ Texture) {constB=2}
  n34 = ScalarParameter() {parameterName=LL_UsePreskinnedPosition}
  n35 = LinearInterpolate(
   A=Multiply(A=Subtract(A=@n31), B=ScalarParameter() {defaultValue=0.1, parameterName=LL_GradientDistortionStrength})
   B=Multiply(A=Subtract(A=@n33), B=ScalarParameter() {defaultValue=0.1, parameterName=LL_WP_GradientDistortionStrength})
   Alpha=@n34
  )
  n36 = Subtract(
   A=ChannelMaskParameter(None=Add(A=@n30, B=@n35)) {maskChannel=Blue, parameterName=LL_GradientDirection, defaultValue=<rgba 0 0 1 0>}.R
   B=Multiply(A=Time(), B=ScalarParameter() {defaultValue=0.01, parameterName=LL_GradientPanSpeed}) {constB=0.05}
  )
  n37 = Multiply(A=Frac(None=@n36), B=ScalarParameter() {defaultValue=4, parameterName=LL_GradientDivisions}) {constB=4}
  ConstantBiasScale(None=Sine(None=@n37))
  ConstantBiasScale(None=Sine(None=Add(A=@n36) {constB=0.75}))
  Frac(None=@n36)
  n38 = CurveAtlasRowParameter(CurveTime=Frac(None=@n37)) {
   atlas=&AthenaCurveAtlas01
   curve=/Game/Characters/Player/Female/Medium/Bodies/F_MED_Loofah/Textures/T_Loofah_Colors_02
   defaultValue=211
   parameterName=LL_COlorCurve
  }
  n39 = LinearInterpolate(
   A=TextureSample(UVs=@n25) {samplerType=SAMPLERTYPE_Normal, texture=/Game/Global/Textures/Default/Noise/T_SwirlNoise_N}.RGB
   B=MaterialFunctionCall(
    TextureObject=TextureObject() {samplerType=SAMPLERTYPE_Normal, texture=&T_SwirlNoise_N}
    TextureSize=@n32.RGB
    WorldPosition=@n12
    World Space Normal=@n10.PS
   ) {materialFunction=&WorldAlignedTexture}.XYZ Texture
   Alpha=@n34
  )
  n40 = DotProduct(A=CameraVectorWS(), B=Transform(None=@n39))
  n41 = Clamp(None=Power(Base=OneMinus(None=@n40)))
  n42 = Multiply(
   A=@n38
   B=LinearInterpolate(
    A=ScalarParameter() {defaultValue=0.22, parameterName=LL_EmissiveEdgeStrength}
    B=ScalarParameter() {defaultValue=1.5, parameterName=LL_EmissiveStrength}
    Alpha=@n41
   )
  )
  n43 = SetMaterialAttributes!(
   MaterialAttributes=@n2.MaterialAttributes
   EmissiveColor=Add(A=@n2.EmissiveColor, B=@n42)
   Specular=ScalarParameter() {defaultValue=1, parameterName=LL_Specular}
   Roughness=ScalarParameter() {defaultValue=0.25, parameterName=LL_Roughness, sliderMin=1}
   BaseColor=Multiply(A=@n38, B=@n41)
   Normal=MaterialFunctionCall(BaseNormal=@n2.Normal, AdditionalNormal=@n39) {materialFunction=&BlendAngleCorrectedNormals}.Result
  ) {attributeSetTypes [
    #1
    #3
    #4
    #5
    #2
   ]}
  n44 = FunctionInput(Preview=Constant()) {bUsePreviewValueAsDefault=true, inputName=Mask}
  FunctionOutput(None=BlendMaterialAttributes(A=@n2.MaterialAttributes, B=@n43, Alpha=ChannelMaskParameter(None=@n44) {parameterName=LL_MaskChannel}.R))
  Multiply(A=@n42, B=Power(Base=Multiply(A=@n3, B=@n4)))
  n45 = Distance(A=@n27, B=Constant() {r=0.5})
  n46 = ScalarParameter() {defaultValue=0.05, parameterName=LL_EdgeStrength}
  n47 = Clamp(None=OneMinus(None=Divide(A=@n45, B=@n46)))
  Multiply(A=Clamp(None=Add(A=@n28, B=@n47)))
  TextureSample() {samplerType=SAMPLERTYPE_Normal, texture=/BRCosmetics/Characters/Player/Female/Medium/Bodies/F_MED_Loofah/Textures/T_F_MED_Loofah_Body_N}
  VertexNormalWS()
 ]}
