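# Physics asset for the dynamic collar and drawstrings on the female medium Loofah cosmetic body, built entirely from capsules.
# Four collar flaps have tight swing and twist limits against small MSTR anchor bodies, which link to spine_05 with free angular motion.
# Two-segment left and right string chains hang from the front collar through their own anchors.
# Head, upper spine and upper-arm capsules supply the surrounding body shapes.
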PhysicsAsset /BRCosmetics/Characters/Player/Female/Medium/Bodies/F_MED_Loofah/Meshes/F_MED_Loofah_Dyn_collar
bodies {
 dyn_collar_bk [{shapeType=Capsule, length=2.75, radius=1.25, center=<1.9887 0 0.0001>, rotation=<rot 90 0 0>}]
 dyn_collar_bk_MSTR [{shapeType=Capsule, length=0.505, radius=0.505}]
 dyn_collar_fr [{shapeType=Capsule, length=2.75, radius=1.25, center=<1.9887 0 0.0001>, rotation=<rot 90 0 0>}]
 dyn_collar_fr_MSTR [{shapeType=Capsule, length=0.505, radius=0.505}]
 dyn_collar_l [{shapeType=Capsule, length=3, radius=1.5, center=<3.0844 -1.5608 0.0001>, rotation=<rot 90 -165.9638 115.49>}]
 dyn_collar_l_MSTR [{shapeType=Capsule, length=0.505, radius=0.505}]
 dyn_collar_r [{shapeType=Capsule, length=8, radius=1.5, center=<0.2673 4.1403 -0.2714>, rotation=<rot 77.5252 161.1329 -137.5089>}]
 dyn_collar_r_MSTR [{shapeType=Capsule, length=0.505, radius=0.505}]
 dyn_string_l_01 [{shapeType=Capsule, length=3.4375, radius=0.5625, shapeName=dyn_string_l_01_capsule, center=<2 0 0>, rotation=<rot 90 0 0>}]
 dyn_string_l_02 [{shapeType=Capsule, length=3.4375, radius=0.5625, shapeName=dyn_string_l_02_capsule, center=<2.566 0 0.0999>, rotation=<rot 80 180 180>}]
 dyn_string_l_MSTR [{shapeType=Capsule, length=0.505, radius=0.505, shapeName=dyn_string_l_MSTR_capsule, rotation=<rot 0 -90 0>}]
 dyn_string_r_01 [{shapeType=Capsule, length=3.4375, radius=0.5625, shapeName=dyn_string_r_01_capsule, center=<2 0 0>, rotation=<rot 90 0 0>}]
 dyn_string_r_02 [{
   shapeType=Capsule
   length=3.4375
   radius=0.5625
   shapeName=dyn_string_r_02_capsule
   center=<2.4568 0.4052 -0.4397>
   rotation=<rot 46.73 -84.1632 -85.1903>
  }]
 dyn_string_r_MSTR [{shapeType=Capsule, length=0.505, radius=0.505, shapeName=dyn_string_r_MSTR_capsule, rotation=<rot 0 -90 0>}]
 head [{shapeType=Capsule, length=7.4688, radius=5.6, shapeName=head_capsule, center=<0.5445 3.4826 0>, rotation=<rot 90 0 0>}]
 spine_04 [{shapeType=Capsule, length=16.5735, radius=8.6678, shapeName=spine_04_capsule, center=<2.9931 3.0018 0>, rotation=<rot 0.4043 -90 -0.6181>}]
 spine_05 [{shapeType=Capsule, length=16.5735, radius=8.6678, shapeName=spine_05_capsule, center=<0.9209 0.3204 0.1377>}]
 upperarm_l [{shapeType=Capsule, length=20, radius=4.5, shapeName=upperarm_l_capsule, center=<9 0 -1>, rotation=<rot 90 0 0>}]
 upperarm_r [{shapeType=Capsule, length=20, radius=4.5, shapeName=upperarm_r_capsule, center=<-12 0 0>, rotation=<rot 90 0 0>}]
}
constraints [
 {bone1Name=dyn_collar_fr, bone2Name=dyn_collar_fr_MSTR, swing1LimitDegrees=15, swing2LimitDegrees=15, twistLimitDegrees=2}
 {bone1Name=dyn_collar_l, bone2Name=dyn_collar_l_MSTR, swing1LimitDegrees=15, swing2LimitDegrees=15, twistLimitDegrees=2}
 {bone1Name=dyn_collar_bk, bone2Name=dyn_collar_bk_MSTR, swing1LimitDegrees=15, swing2LimitDegrees=15, twistLimitDegrees=2}
 {bone1Name=dyn_collar_r, bone2Name=dyn_collar_r_MSTR, swing1LimitDegrees=10, swing2LimitDegrees=30, twistLimitDegrees=2}
 {
  bone1Name=upperarm_l
  bone2Name=spine_05
  swing1LimitDegrees=45
  swing1Motion=Free
  swing2LimitDegrees=45
  swing2Motion=Free
  twistLimitDegrees=45
  twistMotion=Free
 }
 {
  bone1Name=upperarm_r
  bone2Name=spine_05
  swing1LimitDegrees=45
  swing1Motion=Free
  swing2LimitDegrees=45
  swing2Motion=Free
  twistLimitDegrees=45
  twistMotion=Free
 }
 {
  bone1Name=head
  bone2Name=spine_05
  swing1LimitDegrees=45
  swing1Motion=Free
  swing2LimitDegrees=45
  swing2Motion=Free
  twistLimitDegrees=45
  twistMotion=Free
 }
 {
  bone1Name=dyn_collar_fr_MSTR
  bone2Name=spine_05
  swing1LimitDegrees=45
  swing1Motion=Free
  swing2LimitDegrees=45
  swing2Motion=Free
  twistLimitDegrees=45
  twistMotion=Free
 }
 {
  bone1Name=dyn_collar_l_MSTR
  bone2Name=spine_05
  swing1LimitDegrees=45
  swing1Motion=Free
  swing2LimitDegrees=45
  swing2Motion=Free
  twistLimitDegrees=45
  twistMotion=Free
 }
 {
  bone1Name=dyn_collar_bk_MSTR
  bone2Name=spine_05
  swing1LimitDegrees=45
  swing1Motion=Free
  swing2LimitDegrees=45
  swing2Motion=Free
  twistLimitDegrees=45
  twistMotion=Free
 }
 {
  bone1Name=dyn_collar_r_MSTR
  bone2Name=spine_05
  swing1LimitDegrees=45
  swing1Motion=Free
  swing2LimitDegrees=45
  swing2Motion=Free
  twistLimitDegrees=45
  twistMotion=Free
 }
 {
  bone1Name=spine_05
  bone2Name=spine_04
  swing1LimitDegrees=45
  swing1Motion=Free
  swing2LimitDegrees=45
  swing2Motion=Free
  twistLimitDegrees=45
  twistMotion=Free
 }
 {
  bone1Name=dyn_string_l_MSTR
  bone2Name=dyn_collar_fr
  swing1LimitDegrees=45
  swing1Motion=Free
  swing2LimitDegrees=45
  swing2Motion=Free
  twistLimitDegrees=45
  twistMotion=Free
 }
 {bone1Name=dyn_string_l_01, bone2Name=dyn_string_l_MSTR, swing1LimitDegrees=25, swing2LimitDegrees=25, twistLimitDegrees=5}
 {bone1Name=dyn_string_l_02, bone2Name=dyn_string_l_01, swing1LimitDegrees=25, swing2LimitDegrees=25, twistLimitDegrees=5}
 {
  bone1Name=dyn_string_r_MSTR
  bone2Name=dyn_collar_fr
  swing1LimitDegrees=45
  swing1Motion=Free
  swing2LimitDegrees=45
  swing2Motion=Free
  twistLimitDegrees=45
  twistMotion=Free
 }
 {bone1Name=dyn_string_r_01, bone2Name=dyn_string_r_MSTR, swing1LimitDegrees=25, swing2LimitDegrees=25, twistLimitDegrees=5}
 {bone1Name=dyn_string_r_02, bone2Name=dyn_string_r_01, swing1LimitDegrees=25, swing2LimitDegrees=25, twistLimitDegrees=5}
]
