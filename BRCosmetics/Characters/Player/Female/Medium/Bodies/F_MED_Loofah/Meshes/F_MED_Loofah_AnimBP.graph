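# Post-process-style AnimBP for the F_MED_Loofah body, layering secondary motion onto the pose copied from the parent mesh.
# Control Rig feeds AnimDynamics heels, thigh-driven pelvis/thigh corrective offsets, RigidBody collar and string constraints into a cached pose.
# With back bling, detected once, head rotation drives body_MORPH through a driver bone.
# Update derives alphas from reference-pose bone deltas; shoe-tongue nodes and their foot-alpha logic are disconnected.
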AnimBlueprint /BRCosmetics/Characters/Player/Female/Medium/Bodies/F_MED_Loofah/Meshes/F_MED_Loofah_AnimBP
target_skeleton=&/F_MED_Loofah_Skeleton
graphs [
 AnimGraph {nodes [
   n1 = "Transform (Modify) Bone - Bone: dyn_shoe_tongue_r_MSTR"(Alpha="Get foot   Rotate_R"() -> ("foot   Rotate_R":double), Translation="0,0,0", Rotation="-60,0,0", Scale="1,1,1") {
    rotationMode=BMM_Additive
    rotationSpace=BCS_BoneSpace
    scaleMode=BMM_Additive
    scaleSpace=BCS_BoneSpace
    translationMode=BMM_Additive
    translationSpace=BCS_BoneSpace
    alphaBoolBlend {blendOption=Linear}
    boneToModify {boneName=dyn_shoe_tongue_r_MSTR}
    rotation {pitch=-60}
   }
   "Transform (Modify) Bone - Bone: dyn_shoe_tongue_l_MSTR"(ComponentPose=@n1, Alpha=Get foot Rotate L() -> (foot Rotate L:double), Translation="-1,0,0", Rotation="-40,0,0", Scale="1,1,1") {
    rotationMode=BMM_Additive
    rotationSpace=BCS_BoneSpace
    scaleMode=BMM_Additive
    scaleSpace=BCS_BoneSpace
    translationMode=BMM_Additive
    translationSpace=BCS_BoneSpace
    alphaBoolBlend {blendOption=Linear}
    boneToModify {boneName=dyn_shoe_tongue_l_MSTR}
    rotation {pitch=-40}
    translation {x=-1}
   }
   n2 = "Transform (Modify) Bone - Bone: dyn_morph_driver"(
    ComponentPose=Local To Component(LocalPose=Use cached pose 'baseCache'())
    Alpha=Get Head Gross Rot() -> (Head Gross Rot:double)
    Translation="0,0,0"
    Rotation="0,0,1"
    Scale="1,1,1"
   ) {
    rotationMode=BMM_Replace
    rotationSpace=BCS_BoneSpace
    alphaBoolBlend {blendOption=Linear}
    boneToModify {boneName=dyn_morph_driver}
    rotation {roll=1}
   }
   n3 = "body_MORPH = remap(dyn_morph_driver.rotateX) - Bone Driven Controller"!(ComponentPose:Component Space Pose Link Structure=@n2, Alpha:float=1) -> (Pose:Component Space Pose Link Structure) {
    is=title AnimGraphNode_BoneDrivenController
    bUseRange=true
    destinationMode=MorphTarget
    parameterName=body_MORPH
    rangeMin=0
    sourceComponent=RotationX
    alphaBoolBlend {blendOption=Linear}
    sourceBone {boneName=dyn_morph_driver}
   }
   Output Pose(Result=Blend Poses by bool(BlendPose_0=Component To Local(ComponentPose=@n3), BlendPose_1=Use cached pose 'baseCache'(), bActiveValue=Get isWearingBackbling()) {blendType=Linear}) {layerGroup=None}
   n4 = Control Rig(Source=Copy Pose From Mesh() {bUseAttachedParent=true}) {
    alphaBoolBlend {blendOption=Linear}
    controlRigAssetReference {blueprintRigClass=&/F_MED_Loofah_CtrlRig.F_MED_Loofah_CtrlRig_C}
   }
   n5 = Anim Dynamics - Bone: dyn_shoe_heel_r_02(ComponentPose=Local To Component(LocalPose=@n4), Alpha=1, ExternalForce="0,0,0") {
    angularDampingOverride=0.999
    angularSpringConstant=3000
    bAngularSpring=true
    bOverrideAngularDamping=true
    bUseGravityOverride=true
    gravityScale=0.35
    simulationSpace=RootRelative
    alphaBoolBlend {blendOption=Linear}
    boundBone {boneName=dyn_shoe_heel_r_02}
    componentAppliedLinearAccClamp=<10000 10000 10000>
    componentLinearAccScale=<0.2 0.8 0.2>
    componentLinearVelScale=<0.2 0.8 0.2>
    gravityOverride {z=-980}
    physicsBodyDefinitions [{
      boundBone {boneName=dyn_shoe_heel_r_02}
      boxExtents=<20 20 20>
      constraintSetup {angularLimitsMax=<0 30 0>, angularLimitsMin=<0 -10 0>, angularTarget=<1 0 0>}
      localJointOffset=<1 0 0>
     }]
   }
   n6 = Anim Dynamics - Bone: dyn_shoe_heel_l_02(ComponentPose=@n5, Alpha=1, ExternalForce="0,0,0") {
    angularDampingOverride=0.999
    angularSpringConstant=3000
    bAngularSpring=true
    bOverrideAngularDamping=true
    bUseGravityOverride=true
    gravityScale=0.35
    simulationSpace=RootRelative
    alphaBoolBlend {blendOption=Linear}
    boundBone {boneName=dyn_shoe_heel_l_02}
    componentAppliedLinearAccClamp=<10000 10000 10000>
    componentLinearAccScale=<0.2 0.8 0.2>
    componentLinearVelScale=<0.2 0.8 0.2>
    gravityOverride {z=-980}
    physicsBodyDefinitions [{
      boundBone {boneName=dyn_shoe_heel_l_02}
      boxExtents=<20 20 20>
      constraintSetup {angularLimitsMax=<0 30 0>, angularLimitsMin=<0 -10 0>, angularTarget=<1 0 0>}
      localJointOffset=<1 0 0>
     }]
   }
   n7 = Get thigh_rot_l()
   n8 = "Transform (Modify) Bone - Bone: pelvis_l"(ComponentPose=@n6, Alpha=@n7, Translation="-3.025,-0.930,0.000", Rotation="0,0,0", Scale="1,1,1") {
    rotationSpace=BCS_BoneSpace
    scaleMode=BMM_Additive
    scaleSpace=BCS_BoneSpace
    translationMode=BMM_Additive
    translationSpace=BCS_BoneSpace
    alphaBoolBlend {blendOption=Linear}
    boneToModify {boneName=pelvis_l}
    translation=<-3.0253 -0.93 0>
   }
   n9 = "Transform (Modify) Bone - Bone: pelvis_mid_l"(ComponentPose=@n8, Alpha=@n7, Translation="-1.461,0.886,0.000", Rotation="0,0,0", Scale="1,1,1") {
    rotationSpace=BCS_BoneSpace
    scaleMode=BMM_Additive
    scaleSpace=BCS_BoneSpace
    translationMode=BMM_Additive
    translationSpace=BCS_BoneSpace
    alphaBoolBlend {blendOption=Linear}
    boneToModify {boneName=pelvis_mid_l}
    translation=<-1.461 0.886>
   }
   n10 = "Transform (Modify) Bone - Bone: pelvis_side_l"(ComponentPose=@n9, Alpha=@n7, Translation="-1.165,0.593,0.240", Rotation="0,0,0", Scale="1,1,1") {
    rotationSpace=BCS_BoneSpace
    scaleMode=BMM_Additive
    scaleSpace=BCS_BoneSpace
    translationMode=BMM_Additive
    translationSpace=BCS_BoneSpace
    alphaBoolBlend {blendOption=Linear}
    boneToModify {boneName=pelvis_side_l}
    translation=<-1.165 0.593 0.24>
   }
   n11 = "Transform (Modify) Bone - Bone: thigh_volume_bk_l"(ComponentPose=@n10, Alpha=@n7, Translation="2.860,-3.774,0.409", Rotation="0,0,0", Scale="1,1,1") {
    rotationSpace=BCS_BoneSpace
    scaleMode=BMM_Additive
    scaleSpace=BCS_BoneSpace
    translationMode=BMM_Additive
    translationSpace=BCS_BoneSpace
    alphaBoolBlend {blendOption=Linear}
    boneToModify {boneName=thigh_volume_bk_l}
    translation=<2.86 -3.774 0.409>
   }
   n12 = "Transform (Modify) Bone - Bone: thigh_volume_fr_l"(ComponentPose=@n11, Alpha=@n7, Translation="-2.144,7.226,-0.000", Rotation="0,0,0", Scale="1,1,1") {
    rotationSpace=BCS_BoneSpace
    scaleMode=BMM_Additive
    scaleSpace=BCS_BoneSpace
    translationMode=BMM_Additive
    translationSpace=BCS_BoneSpace
    alphaBoolBlend {blendOption=Linear}
    boneToModify {boneName=thigh_volume_fr_l}
    translation=<-2.144 7.226>
   }
   n13 = Get thigh_rot_r()
   n14 = "Transform (Modify) Bone - Bone: pelvis_r"(ComponentPose=@n12, Alpha=@n13, Translation="-3.022,-0.930,0.000", Rotation="0,0,0", Scale="1,1,1") {
    rotationSpace=BCS_BoneSpace
    scaleMode=BMM_Additive
    scaleSpace=BCS_BoneSpace
    translationMode=BMM_Additive
    translationSpace=BCS_BoneSpace
    alphaBoolBlend {blendOption=Linear}
    boneToModify {boneName=pelvis_r}
    translation=<-3.022 -0.93>
   }
   n15 = "Transform (Modify) Bone - Bone: pelvis_mid_r"(ComponentPose=@n14, Alpha=@n13, Translation="-1.461,-1.001,-0.000", Rotation="0,0,0", Scale="1,1,1") {
    rotationSpace=BCS_BoneSpace
    scaleMode=BMM_Additive
    scaleSpace=BCS_BoneSpace
    translationMode=BMM_Additive
    translationSpace=BCS_BoneSpace
    alphaBoolBlend {blendOption=Linear}
    boneToModify {boneName=pelvis_mid_r}
    translation=<-1.461 -1.001>
   }
   n16 = "Transform (Modify) Bone - Bone: pelvis_side_r"(ComponentPose=@n15, Alpha=@n13, Translation="-1.165,0.000,-0.817", Rotation="0,0,0", Scale="1,1,1") {
    rotationSpace=BCS_BoneSpace
    scaleMode=BMM_Additive
    scaleSpace=BCS_BoneSpace
    translationMode=BMM_Additive
    translationSpace=BCS_BoneSpace
    alphaBoolBlend {blendOption=Linear}
    boneToModify {boneName=pelvis_side_r}
    translation {x=-1.165, z=-0.817}
   }
   n17 = "Transform (Modify) Bone - Bone: thigh_volume_bk_r"(ComponentPose=@n16, Alpha=@n13, Translation="2.860,5.043,0.000", Rotation="0,0,0", Scale="1,1,1") {
    rotationSpace=BCS_BoneSpace
    scaleMode=BMM_Additive
    scaleSpace=BCS_BoneSpace
    translationMode=BMM_Additive
    translationSpace=BCS_BoneSpace
    alphaBoolBlend {blendOption=Linear}
    boneToModify {boneName=thigh_volume_bk_r}
    translation=<2.86 5.043>
   }
   n18 = "Transform (Modify) Bone - Bone: thigh_volume_fr_r"(ComponentPose=@n17, Alpha=@n13, Translation="-2.144,-7.226,-0.000", Rotation="0,0,0", Scale="1,1,1") {
    rotationSpace=BCS_BoneSpace
    scaleMode=BMM_Additive
    scaleSpace=BCS_BoneSpace
    translationMode=BMM_Additive
    translationSpace=BCS_BoneSpace
    alphaBoolBlend {blendOption=Linear}
    boneToModify {boneName=thigh_volume_fr_r}
    translation=<-2.144 -7.226>
   }
   n19 = RigidBody(ComponentPose=@n18) {
    bClampLinearTranslationLimitToRefPose=true
    bOverrideWorldGravity=true
    overridePhysicsAsset=&/F_MED_Loofah_Dyn_collar
    simulationSpace=BaseBoneSpace
    alphaBoolBlend {blendOption=Linear}
    baseBoneRef {boneName=spine_05}
    componentAppliedLinearAccClamp=<3000 3000 3000>
    componentLinearAccScale=<0.4 0.3 0.3>
    componentLinearVelScale=<0.4 0.3 0.3>
    overrideWorldGravity {z=-980}
   }
   n20 = "Constraint - dyn_string_r_MSTR (PPP)"!(ComponentPose=@n19, Alpha=1, ConstraintWeights_0=0.2892, ConstraintWeights_1:float=0.096, ConstraintWeights_2:float=0.6147) -> (Pose) {
    alphaBoolBlend {blendOption=Linear}
    boneToModify {boneName=dyn_string_r_MSTR}
    constraintSetup [{targetBone {boneName=dyn_collar_r}}, {targetBone {boneName=spine_05}}, {targetBone {boneName=dyn_collar_fr}}]
    constraintWeights [1,1,1]
   }
   n21 = "Constraint - dyn_string_l_MSTR (PPP)"!(ComponentPose=@n20, Alpha=1, ConstraintWeights_0=0.0938, ConstraintWeights_1:float=0.3217, ConstraintWeights_2:float=0.5846) -> (Pose) {
    alphaBoolBlend {blendOption=Linear}
    boneToModify {boneName=dyn_string_l_MSTR}
    constraintSetup [{targetBone {boneName=spine_05}}, {targetBone {boneName=dyn_collar_l}}, {targetBone {boneName=dyn_collar_fr}}]
    constraintWeights [1,1,1]
   }
   Save cached pose 'baseCache'(Pose=Component To Local(ComponentPose=@n21)) {cacheName=baseCache}
  ]}
 EventGraph {nodes [
   Event BlueprintUpdateAnimation()
   n1 = Get Skel_Mesh()
   > Set thigh_rot_r(thigh_rot_r=GetAlphaFromReferenceTransformDelta(SkeletalMeshComponent=@n1, BoneName=thigh_r, BaseName=pelvis, InputRange=(X=0,Y=-90), TwistAxis="0,0,1"))
   > Set thigh_rot_l(thigh_rot_l=GetAlphaFromReferenceTransformDelta(SkeletalMeshComponent=@n1, BoneName=thigh_l, BaseName=pelvis, InputRange=(X=-15,Y=-110), TwistAxis="0,0,1"))
   > n2 = Set Head Rot Up(Head Rot Up:double=GetAlphaFromReferenceTransformDelta(SkeletalMeshComponent=@n1, BoneName=head, BaseName=spine_05, InputRange=(X=5,Y=-45), TwistAxis="0,0,1")) -> (Output_Get:double)
   > n3 = Set Head Rot L(Head Rot L:double=GetAlphaFromReferenceTransformDelta(SkeletalMeshComponent=@n1, BoneName=head, BaseName=spine_05, DeltaComponentType=EulerX, InputRange=(X=0,Y=40), TwistAxis="0,0,1")) -> (Output_Get:double)
   > Set Head Gross Rot(Head Gross Rot:double=float * float!(A=@n3, B=@n2, C:double=1.5) -> (ReturnValue)) -> (Output_Get:double)
   n4 = Event OnCharacterCustomizationLoaded()
   Do Once(execute=[@n4.then, @n6.then])
   n5 = CheckForBackbling()
   > Set isWearingBackbling(isWearingBackbling=OR Boolean(A=@n5, B=@n5.bIsBackblingLong))
   > Set Skel_Mesh(Skel_Mesh=GetOwnerSkeletalMeshComponent())
   n6 = Event BlueprintInitializeAnimation()
   n7 = Sequence!(execute) -> (then_0, then_1, then_2:Exec, then_3:Exec, then_4:Exec)
   Set foot Rotate L(
    execute=@n7.then_1
    foot Rotate L:double=GetAlphaFromReferenceTransformDelta(
     SkeletalMeshComponent=Get Skel_Mesh()
     BoneName=calf_l
     BaseName=foot_l
     DeltaComponentType=QuaternionTwist
     InputRange=(X=0,Y=90)
     TwistAxis="0,0,1"
    )
   ) -> (Output_Get:double)
   "Set foot   Rotate_R"(
    execute=@n7.then_2
    "foot   Rotate_R":double=GetAlphaFromReferenceTransformDelta(
     SkeletalMeshComponent=Get Skel_Mesh()
     BoneName=calf_r
     BaseName=foot_r
     DeltaComponentType=QuaternionTwist
     InputRange=(X=0,Y=90)
     TwistAxis="0,0,1"
    )
   ) -> (Output_Get:double)
  ]}
]
variables [
 "foot   Rotate_R" {type=number}
 foot Rotate L {type=number}
 Head Gross Rot {type=number}
 Head Rot L {type=number}
 Head Rot Up {type=number}
 isWearingBackbling {type=boolean}
 Skel_Mesh {type=SkeletalMeshComponent}
 thigh_rot_l {type=number}
 thigh_rot_r {type=number}
]
