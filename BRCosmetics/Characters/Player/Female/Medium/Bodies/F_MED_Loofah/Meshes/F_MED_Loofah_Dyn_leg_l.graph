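# Physics asset for the left leg of the F_MED_Loofah female player body cosmetic, covering the dynamic shoe heel and tongue bones.
# A calf capsule roots two capsule chains, each starting at a small master body: a two-segment heel and a single-segment tongue.
# Master bodies have free swing and twist relative to the calf, while segments below them use tight angular limits to restrict their motion.
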PhysicsAsset /BRCosmetics/Characters/Player/Female/Medium/Bodies/F_MED_Loofah/Meshes/F_MED_Loofah_Dyn_leg_l
bodies {
 calf_l [{shapeType=Capsule, length=14.6973, radius=2.3936, center=<-34.915 -3.4224 -1.3199>, rotation=<rot 90 -1.7899 3.2102>}]
 dyn_shoe_heel_l_01 [{shapeType=Capsule, length=3.125, radius=0.875, center=<1 0 0.0001>, rotation=<rot 90 2.8662 92.9805>}]
 dyn_shoe_heel_l_02 [{shapeType=Capsule, length=3.125, radius=0.875, center=<1.8807 0 0.0001>, rotation=<rot 90 0 0>}]
 dyn_shoe_heel_l_MSTR [{shapeType=Capsule, length=0.505, radius=0.505}]
 dyn_shoe_tongue_l [{shapeType=Capsule, length=3.125, radius=0.875, center=<3.3424 -0.0239 0.0001>, rotation=<rot 90 180 90.584>}]
 dyn_shoe_tongue_l_MSTR [{shapeType=Capsule, length=0.505, radius=0.505}]
}
constraints [
 {bone1Name=dyn_shoe_tongue_l, bone2Name=dyn_shoe_tongue_l_MSTR, swing1LimitDegrees=20, swing2LimitDegrees=45, twistLimitDegrees=5}
 {bone1Name=dyn_shoe_heel_l_01, bone2Name=dyn_shoe_heel_l_MSTR, swing1LimitDegrees=20, swing2LimitDegrees=30, twistLimitDegrees=5}
 {bone1Name=dyn_shoe_heel_l_02, bone2Name=dyn_shoe_heel_l_01, swing1LimitDegrees=10, swing2LimitDegrees=20, twistLimitDegrees=5}
 {
  bone1Name=dyn_shoe_tongue_l_MSTR
  bone2Name=calf_l
  swing1LimitDegrees=45
  swing1Motion=Free
  swing2LimitDegrees=45
  swing2Motion=Free
  twistLimitDegrees=45
  twistMotion=Free
 }
 {
  bone1Name=dyn_shoe_heel_l_MSTR
  bone2Name=calf_l
  swing1LimitDegrees=45
  swing1Motion=Free
  swing2LimitDegrees=45
  swing2Motion=Free
  twistLimitDegrees=45
  twistMotion=Free
 }
]
