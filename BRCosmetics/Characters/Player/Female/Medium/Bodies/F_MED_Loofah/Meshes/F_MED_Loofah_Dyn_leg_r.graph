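# Physics asset for the right leg of the F_MED_Loofah female medium player body, simulating secondary motion on the shoe's heel and tongue.
# A calf capsule anchors small master bodies for heel and tongue through constraints with free swing and twist.
# A two-bone heel chain and a single tongue bone hang off those masters, with moderate swing limits and tight twist limits.
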PhysicsAsset /BRCosmetics/Characters/Player/Female/Medium/Bodies/F_MED_Loofah/Meshes/F_MED_Loofah_Dyn_leg_r
bodies {
 calf_r [{shapeType=Capsule, length=16, radius=2.2061, center=<37.9891 3.1534 1.3199>, rotation=<rot 90 180 -174.9999>}]
 dyn_shoe_heel_r_01 [{shapeType=Capsule, length=3.125, radius=0.875, center=<1 0 0.0001>, rotation=<rot 90 -2.8662 90.3313>}]
 dyn_shoe_heel_r_02 [{shapeType=Capsule, length=3.125, radius=0.875, center=<1.8807 0 0.0001>, rotation=<rot 90 0 0>}]
 dyn_shoe_heel_r_MSTR [{shapeType=Capsule, length=0.505, radius=0.505}]
 dyn_shoe_tongue_r [{shapeType=Capsule, length=3.125, radius=0.875, center=<2.3845 -0.0439 0.0001>, rotation=<rot 90 165.9638 77.7789>}]
 dyn_shoe_tongue_r_MSTR [{shapeType=Capsule, length=0.505, radius=0.505}]
}
constraints [
 {bone1Name=dyn_shoe_tongue_r, bone2Name=dyn_shoe_tongue_r_MSTR, swing1LimitDegrees=20, swing2LimitDegrees=45, twistLimitDegrees=5}
 {bone1Name=dyn_shoe_heel_r_01, bone2Name=dyn_shoe_heel_r_MSTR, swing1LimitDegrees=20, swing2LimitDegrees=30, twistLimitDegrees=5}
 {bone1Name=dyn_shoe_heel_r_02, bone2Name=dyn_shoe_heel_r_01, swing1LimitDegrees=20, swing2LimitDegrees=30, twistLimitDegrees=5}
 {
  bone1Name=dyn_shoe_tongue_r_MSTR
  bone2Name=calf_r
  swing1LimitDegrees=45
  swing1Motion=Free
  swing2LimitDegrees=45
  swing2Motion=Free
  twistLimitDegrees=45
  twistMotion=Free
 }
 {
  bone1Name=dyn_shoe_heel_r_MSTR
  bone2Name=calf_r
  swing1LimitDegrees=45
  swing1Motion=Free
  swing2LimitDegrees=45
  swing2Motion=Free
  twistLimitDegrees=45
  twistMotion=Free
 }
]
