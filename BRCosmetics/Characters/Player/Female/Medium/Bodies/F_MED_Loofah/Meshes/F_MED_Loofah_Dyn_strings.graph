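# Physics asset for the F_MED_Loofah female body's dynamic front collar and two hanging strings, giving them secondary motion.
# The collar swings within narrow limits from a master bone, which is linked to spine_05 by a free constraint.
# Each string has a master bone locked to the collar, followed by two capsule segments with limited swing and twist.
# Larger spine, right upper arm and left thigh capsules provide body volumes, likely for collision.
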PhysicsAsset /BRCosmetics/Characters/Player/Female/Medium/Bodies/F_MED_Loofah/Meshes/F_MED_Loofah_Dyn_strings
bodies {
 dyn_collar_fr [{shapeType=Capsule, length=2.75, radius=1.25, center=<1.9887 0 0.0001>, rotation=<rot 90 0 0>}]
 dyn_collar_fr_MSTR [{shapeType=Capsule, length=0.505, radius=0.505}]
 dyn_string_l_01 [{shapeType=Capsule, length=3.4375, radius=0.5625, center=<2 0 0>, rotation=<rot 90 0 0>}]
 dyn_string_l_02 [{shapeType=Capsule, length=3.4375, radius=0.5625, center=<2.566 0 0.0999>, rotation=<rot 79.9926 180 180>}]
 dyn_string_l_MSTR [{shapeType=Capsule, length=0.505, radius=0.505}]
 dyn_string_r_01 [{shapeType=Capsule, length=3.4375, radius=0.5625, center=<2 0 0>, rotation=<rot 90 0 0>}]
 dyn_string_r_02 [{shapeType=Capsule, length=3.4375, radius=0.5625, center=<2.4568 0.4052 -0.4397>, rotation=<rot 46.73 -84.1632 -85.1903>}]
 dyn_string_r_MSTR [{shapeType=Capsule, length=0.505, radius=0.505}]
 spine_04 [{shapeType=Capsule, length=16.5735, radius=8.6678, center=<2.9931 3.0018 0>}]
 spine_05 [{shapeType=Capsule, length=16.5735, radius=8.6678, center=<0.9209 0.3204 0>}]
 thigh_l [{shapeType=Capsule, length=30, radius=7, shapeName=thigh_l_capsule, center=<-20 0.3 0>, rotation=<rot 0 -89 -95>}]
 upperarm_r [{shapeType=Capsule, length=20, radius=4.5, shapeName=upperarm_r_capsule, center=<-12 0 0>, rotation=<rot 90 0 0>}]
}
constraints [
 {bone1Name=dyn_collar_fr, bone2Name=dyn_collar_fr_MSTR, swing1LimitDegrees=15, swing2LimitDegrees=15, twistLimitDegrees=2}
 {
  bone1Name=dyn_string_l_MSTR
  bone2Name=dyn_collar_fr
  swing1LimitDegrees=45
  swing1Motion=Locked
  swing2LimitDegrees=45
  swing2Motion=Locked
  twistLimitDegrees=45
  twistMotion=Locked
 }
 {bone1Name=dyn_string_l_01, bone2Name=dyn_string_l_MSTR, swing1LimitDegrees=25, swing2LimitDegrees=25, twistLimitDegrees=5}
 {bone1Name=dyn_string_l_02, bone2Name=dyn_string_l_01, swing1LimitDegrees=25, swing2LimitDegrees=25, twistLimitDegrees=5}
 {
  bone1Name=dyn_string_r_MSTR
  bone2Name=dyn_collar_fr
  swing1LimitDegrees=45
  swing1Motion=Locked
  swing2LimitDegrees=45
  swing2Motion=Locked
  twistLimitDegrees=45
  twistMotion=Locked
 }
 {bone1Name=dyn_string_r_01, bone2Name=dyn_string_r_MSTR, swing1LimitDegrees=25, swing2LimitDegrees=25, twistLimitDegrees=5}
 {bone1Name=dyn_string_r_02, bone2Name=dyn_string_r_01, swing1LimitDegrees=25, swing2LimitDegrees=25, twistLimitDegrees=5}
 {
  bone1Name=dyn_collar_fr_MSTR
  bone2Name=spine_05
  swing1LimitDegrees=45
  swing1Motion=Free
  swing2LimitDegrees=45
  swing2Motion=Free
  twistLimitDegrees=45
  twistMotion=Free
 }
 {
  bone1Name=spine_05
  bone2Name=spine_04
  swing1LimitDegrees=45
  swing1Motion=Free
  swing2LimitDegrees=45
  swing2Motion=Free
  twistLimitDegrees=45
  twistMotion=Free
 }
 {
  bone1Name=upperarm_r
  bone2Name=spine_05
  swing1LimitDegrees=45
  swing1Motion=Free
  swing2LimitDegrees=45
  swing2Motion=Free
  twistLimitDegrees=45
  twistMotion=Free
 }
]
