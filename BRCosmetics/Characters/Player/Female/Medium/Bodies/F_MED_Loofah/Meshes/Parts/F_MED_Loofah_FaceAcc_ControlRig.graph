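# Control Rig for the F_MED_Loofah female face accessory, applying pose-driven corrective offsets to the hat, hat bill and ponytail bones as the head moves.
# A spherical pose reader on the head, with debug drawing on, weights offsets to dyn_hat, dyn_ponytail_MSTR and dyn_hat_bill.
# Head rotation relative to spine_05 becomes Euler angles, whose remapped, unclamped X and Y values weight further ponytail and hat offsets in both directions.
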ControlRigBlueprint /BRCosmetics/Characters/Player/Female/Medium/Bodies/F_MED_Loofah/Meshes/Parts/F_MED_Loofah_FaceAcc_ControlRig
graphs [
 RigVMFunctionLibrary {links [], nodes []}
 RigVMModel {nodes [
   n1 = &RigUnit_BeginExecution()
   n2 = &RigUnit_SphericalPoseReader(
    ActiveRegionScaleFactors=(PositiveWidth=2.2,NegativeWidth=0.5,PositiveHeight=1.4,NegativeHeight=1)
    ActiveRegionSize=0.05
    Debug=(bDrawDebug=true,bDraw2D=false,bDrawLocalAxes=false,DebugScale=25,DebugSegments=20,DebugThickness=0.25)
    DriverAxis=(X=-1,Y=-1,Z=0)
    DriverItem=(Type=Bone,Name="head")
    FalloffRegionScaleFactors=(PositiveWidth=1,NegativeWidth=1.6,PositiveHeight=1,NegativeHeight=1)
    FalloffSize=0.25
    FlipHeightScaling=False
    FlipWidthScaling=False
    OptionalParentItem=(Type=Bone,Name="None")
    RotationOffset=(X=90,Y=0,Z=800)
    ExecutePin=@n1.ExecutePin
   )
   > &RigUnit_ModifyTransforms(
    ItemToModify=((Item=(Type=Bone,Name="None"),Transform=(Rotation=(X=0,Y=0,Z=0,W=1),Translation=(X=0,Y=0,Z=0),Scale3D=(X=1,Y=1,Z=1))),(Item=(Type=Bone,Name="dyn_hat"),Transform=(Rotation=(X=0,Y=0,Z=0.1736,W=0.9848),Translation=(X=2,Y=0,Z=0),Scale3D=(X=1,Y=1,Z=1))))
    WeightMinimum=0
    Weight=@n2.OutputParam
   )
   > &RigUnit_ModifyTransforms(
    ItemToModify="((Item=(Type=Bone,Name=\"None\"),Transform=(Rotation=(X=0,Y=0,Z=0,W=1),Translation=(X=0,Y=0,Z=0),Scale3D=(X=1,Y=1,Z=1))),(Item=(Type=Bone,Name=\"dyn_ponytail_MSTR\"),Transform=(Rotation=(X=0,Y=-0.2588,Z=0,W=0.9659),Translation=(X=0,Y=0,Z=0),Scale3D=(X=1,Y=1,Z= ...(truncated)"
    WeightMinimum=0
    Weight=@n2.OutputParam
   )
   > &RigUnit_ModifyTransforms(
    ItemToModify=((Item=(Type=Bone,Name="None"),Transform=(Rotation=(X=0,Y=0,Z=0,W=1),Translation=(X=0,Y=0,Z=0),Scale3D=(X=1,Y=1,Z=1))),(Item=(Type=Bone,Name="dyn_hat_bill"),Transform=(Rotation=(X=0,Y=0,Z=0.1736,W=0.9848),Translation=(X=0,Y=0,Z=0),Scale3D=(X=1,Y=1,Z=1))))
    WeightMinimum=0
    Weight=@n2.OutputParam
   )
   RigVMFunction_MathQuaternionToEuler = &RigVMFunction_MathQuaternionToEuler(
    RotationOrder=ZYX
    Value=&RigUnit_GetRelativeTransformForItem(Child=(Type=Bone,Name="head"), Parent=(Type=Bone,Name="spine_05"), bChildInitial=False, bParentInitial=False).RelativeTransform.Rotation
   )
   > &RigUnit_ModifyTransforms(
    ItemToModify="((Item=(Type=Bone,Name=\"None\"),Transform=(Rotation=(X=0,Y=0,Z=0,W=1),Translation=(X=0,Y=0,Z=0),Scale3D=(X=1,Y=1,Z=1))),(Item=(Type=Bone,Name=\"dyn_ponytail_01\"),Transform=(Rotation=(X=0,Y=0,Z=0.4226,W=0.9063),Translation=(X=0,Y=0,Z=0),Scale3D=(X=1,Y=1,Z=1)) ...(truncated)"
    WeightMinimum=0
    Weight=&RigVMFunction_MathDoubleRemap(SourceMaximum=-80, SourceMinimum=-20, TargetMaximum=1, TargetMinimum=0, bClamp=False, Value=@RigVMFunction_MathQuaternionToEuler.Result.X).Result
   )
   > &RigUnit_ModifyTransforms(
    ItemToModify=((Item=(Type=Bone,Name="None"),Transform=(Rotation=(X=0,Y=0,Z=0,W=1),Translation=(X=0,Y=0,Z=0),Scale3D=(X=1,Y=1,Z=1))),(Item=(Type=Bone,Name="dyn_hat"),Transform=(Rotation=(X=0,Y=0,Z=0.1305,W=0.9914),Translation=(X=0,Y=0,Z=0),Scale3D=(X=1,Y=1,Z=1))))
    WeightMinimum=0
    Weight=&RigVMFunction_MathDoubleRemap(SourceMaximum=-40, SourceMinimum=0, TargetMaximum=1, TargetMinimum=0, bClamp=False, Value=@RigVMFunction_MathQuaternionToEuler.Result.Y).Result
   )
   Remap = &RigVMFunction_MathDoubleRemap(SourceMaximum=70, SourceMinimum=20, TargetMaximum=1, TargetMinimum=0, bClamp=False, Value=@RigVMFunction_MathQuaternionToEuler.Result.X)
   > &RigUnit_ModifyTransforms(
    ItemToModify=((Item=(Type=Bone,Name="None"),Transform=(Rotation=(X=0,Y=0,Z=0,W=1),Translation=(X=0,Y=0,Z=0),Scale3D=(X=1,Y=1,Z=1))),(Item=(Type=Bone,Name="dyn_ponytail_01"),Transform=(Rotation=(X=0,Y=0,Z=-0.5,W=0.866),Translation=(X=0,Y=0,Z=0),Scale3D=(X=1,Y=1,Z=1))))
    WeightMinimum=0
    Weight=@Remap.Result
   )
   > &RigUnit_ModifyTransforms(
    ItemToModify=((Item=(Type=Bone,Name="None"),Transform=(Rotation=(X=0,Y=0,Z=0,W=1),Translation=(X=0,Y=0,Z=0),Scale3D=(X=1,Y=1,Z=1))),(Item=(Type=Bone,Name="dyn_hat"),Transform=(Rotation=(X=0,Y=0,Z=0.1305,W=0.9914),Translation=(X=0,Y=0,Z=0),Scale3D=(X=1,Y=1,Z=1))))
    WeightMinimum=0
    Weight=&RigVMFunction_MathDoubleRemap(SourceMaximum=40, SourceMinimum=0, TargetMaximum=1, TargetMinimum=0, bClamp=False, Value=@RigVMFunction_MathQuaternionToEuler.Result.Y).Result
   )
  ]}
]
hierarchy [
 root {loc=<0 0 0>, rot=<0 0 0 1>}
 pelvis {parent=root, loc=<0 -0.5301 98.0571>, rot=<0.0127 0.707 -0.0127 -0.707>}
 spine_01 {parent=pelvis, loc=<6.1208 -0.9328 0>, rot=<0 0 0.034 -0.9994>}
 spine_02 {parent=spine_01, loc=<6.5996 -0.2095 0>, rot=<0 0 0.0238 -0.9997>}
 spine_03 {parent=spine_02, loc=<6.4914 -0.1237 0>, rot=<0 0 0.0149 -0.9999>}
 spine_04 {parent=spine_03, loc=<6.9598 0.0794 0>, rot=<0 0 -0.0132 -0.9999>}
 spine_05 {parent=spine_04, loc=<6.9987 0 0>, rot=<0 0 -0.0015 -1>}
 clavicle_l {parent=spine_05, loc=<5.4831 -0.1053 -3.2946>, rot=<0.0465 -0.7289 -0.009 -0.6829>}
 dyn_ponytail_CV3 {parent=clavicle_l, loc=<-2.3155 -16.4921 -2.3358>, rot=<-0.4501 -0.4355 0.5217 -0.5792>}
 upperarm_l {parent=clavicle_l, loc=<9.8658 -1.2136 -0.4105>, rot=<0.0208 -0.486 -0.0242 -0.8734>}
 lowerarm_l {parent=upperarm_l, loc=<26.465 -0.8345 0.259>, rot=<0.0036 0.0161 -0.2298 -0.9731>}
 clavicle_r {parent=spine_05, loc=<5.4831 -0.1053 3.2946>, rot=<0.7289 0.0465 -0.6829 0.009>}
 upperarm_r {parent=clavicle_r, loc=<-9.8658 1.2136 0.4104>, rot=<0.0208 -0.486 -0.0242 -0.8734>}
 lowerarm_r {parent=upperarm_r, loc=<-26.4651 0.8345 -0.259>, rot=<0.0036 0.0161 -0.2298 -0.9731>}
 neck_01 {parent=spine_05, loc=<11.3138 0.1529 0>, rot=<0 0 -0.2003 -0.9797>}
 neck_02 {parent=neck_01, loc=<4.5627 0 0>, rot=<0 0 0.0504 -0.9987>}
 head {parent=neck_02, loc=<5.3914 0 0>, rot=<0 0 0.115 -0.9934>}
 dyn_side_hair_l_01 {parent=head, loc=<5.1915 5.0927 -6.0864>, rot=<0.0162 -0.9941 -0.1033 -0.0288>}
 dyn_side_hair_l_02 {parent=dyn_side_hair_l_01, loc=<5.3045 0 0>, rot=<-0.0112 -0.1021 -0.1089 -0.9887>}
 dyn_side_hair_l_03 {parent=dyn_side_hair_l_02, loc=<6.0959 0 0>, rot=<-0.0183 -0.1025 -0.1749 -0.9791>}
 dyn_side_hair_l_04 {parent=dyn_side_hair_l_01, loc=<4.8097 -1.8154 1.379>, rot=<-0.0763 -0.049 0.0035 -0.9959>}
 dyn_side_hair_l_05 {parent=dyn_side_hair_l_01, loc=<5.3029 -2.7708 -0.4561>, rot=<-0.0763 -0.049 0.0035 -0.9959>}
 dyn_side_hair_r_01 {parent=head, loc=<4.9808 4.508 6.1469>, rot=<0.0288 0.1033 0.9941 -0.0162>}
 dyn_side_hair_r_04 {parent=dyn_side_hair_r_01, loc=<6.3393 1.9554 0.0418>, rot=<0.0763 -0.049 -0.0035 -0.9959>}
 dyn_side_hair_r_02 {parent=dyn_side_hair_r_01, loc=<5.5339 0.7771 -0.1603>, rot=<0.0145 -0.0697 0.1042 -0.992>}
 dyn_side_hair_r_03 {parent=dyn_side_hair_r_02, loc=<6.0958 0 0>, rot=<0.0183 -0.1025 0.1749 -0.9791>}
 dyn_hat {parent=head, loc=<6.0925 0.8971 0>, rot=<0 0 0 -1>}
 dyn_ponytail_MSTR {parent=dyn_hat, loc=<-3.9707 -3.3697 0>, rot=<0.2949 -0.6427 0.6427 -0.2949>}
 dyn_ponytail_00 {parent=dyn_ponytail_MSTR, loc=<0 0 0>, rot=<0 0 0 1>}
 dyn_ponytail_01 {parent=dyn_ponytail_00, loc=<3.0805 0 0>, rot=<0 0 0 1>}
 dyn_ponytail_02 {parent=dyn_ponytail_01, loc=<6.0012 0 0>, rot=<0 0 0 -1>}
 dyn_ponytail_03 {parent=dyn_ponytail_02, loc=<7.5506 0 0>, rot=<0 -0.2209 0 -0.9753>}
 dyn_ponytail_04 {parent=dyn_ponytail_03, loc=<9.5199 0 0>, rot=<0 -0.096 0 -0.9954>}
 dyn_ponytail_05 {parent=dyn_ponytail_04, loc=<9.3888 0 0>, rot=<0 -0.0525 0 -0.9986>}
 dyn_ponytail_06 {parent=dyn_ponytail_05, loc=<8.5208 0 0>, rot=<0 -0.0682 0 -0.9977>}
 dyn_hat_bill {parent=dyn_hat, loc=<-1.7602 -5.4236 0>, rot=<0 0 0 -1>}
 faceAttach {parent=head, loc=<-2.2967 -1.5887 0>, rot=<0.0044 -0.3827 -0.0044 -0.9239>, scale=<0.89 0.89 0.89>}
 dyn_ponytail_CV4 {parent=spine_05, loc=<-4.36 -19.2984 0>, rot=<0.0668 -0.7039 0.7039 -0.0668>}
 hood {parent=spine_05, loc=<11.618 -5.803 0.416>, rot=<0.0284 -0.7065 -0.0284 -0.7065>}
 dyn_ponytail_CV5 {parent=spine_04, loc=<-5.6673 -20.9397 0>, rot=<0.0197 -0.7068 0.7068 -0.0197>}
 thigh_l {parent=pelvis, loc=<-1.0855 -0.2109 -8.7053>, rot=<0.0009 0.0355 0.0155 -0.9993>}
 head_ctrl {controlType=EULER_TRANSFORM, type=Control, loc=<0 0 0>, rot=<0 0 0 1>}
]
