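# Physics asset for the F_MED_Loofah female character's face accessory part, covering a ponytail, side hair strands and a hat bill.
# The ponytail chain is nearly locked at its root and swings more freely from the third segment down, with tight twist limits.
# Limited side-hair chains hang from the head and branch at their first bone, asymmetrically left and right.
# Head, torso and arm capsules provide body collision volumes; their constraints and the hat bill's are angularly free.
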PhysicsAsset /BRCosmetics/Characters/Player/Female/Medium/Bodies/F_MED_Loofah/Meshes/Parts/F_MED_Loofah_FaceAcc_Dyn
bodies {
 clavicle_l [{shapeType=Capsule, length=6.2188, radius=6.7188, center=<4.2524 0.0001 -1.5232>, rotation=<rot 84.6895 0 0>}]
 clavicle_r [{shapeType=Capsule, length=6.2188, radius=6, center=<-4.2524 -0.0001 1.5232>, rotation=<rot 84.6895 0 0>}]
 dyn_hat_bill [{shapeType=Box, extentX=15, extentY=10, extentZ=15, shapeName=dyn_hat_bill_box, center=<1.9991 -3.2551 0>, rotation=<rot 0 70 0>}]
 dyn_ponytail_00 [{shapeType=Capsule, length=0, radius=1.9, shapeName=dyn_ponytail_00_capsule, center=<0.3 0 0>, rotation=<rot 90 0 0>}]
 dyn_ponytail_01 [{shapeType=Capsule, length=2.6, radius=2.1, center=<2.3 0 0>, rotation=<rot 90 0 0>}]
 dyn_ponytail_02 [{shapeType=Capsule, length=3.2, radius=2, center=<3 0 0>, rotation=<rot 90 0 -0.1962>}]
 dyn_ponytail_03 [{shapeType=Capsule, length=2, radius=4, center=<4.2307 0.3678 0>, rotation=<rot 90 0 0>}]
 dyn_ponytail_04 [{shapeType=Capsule, length=0, radius=5.5, center=<4.6701 0.7147 0>, rotation=<rot 90 0 0>}]
 dyn_ponytail_05 [{shapeType=Capsule, length=0, radius=4, center=<4.6701 0.5201 0>, rotation=<rot 90 0 0>}]
 dyn_ponytail_06 [{shapeType=Capsule, length=5.75, radius=2.2188, center=<4.6701 0 0>, rotation=<rot 90 0 0>}]
 dyn_ponytail_MSTR [{shapeType=Capsule, length=0, radius=1.9, center=<0.3 0 0>, rotation=<rot 90 0 0>}]
 dyn_side_hair_l_01 [{shapeType=Box, extentX=5.8125, extentY=4, extentZ=1.625, center=<2.6057 0.1 0>}]
 dyn_side_hair_l_02 [{shapeType=Box, extentX=5, extentY=2.5625, extentZ=1.625, center=<3.2 0.0001 0>}]
 dyn_side_hair_l_03 [{shapeType=Box, extentX=5, extentY=0.75, extentZ=0.8125, center=<2.5121 -0.052 0.1817>, rotation=<rot 1.0607 3.0564 1.4478>}]
 dyn_side_hair_l_04 [{
   shapeType=Capsule
   length=3.4978
   radius=1
   shapeName=dyn_side_hair_l_04_capsule
   center=<2.0061 -2.5473 0.2952>
   rotation=<rot -1.3051 20.3817 -83.846>
  }]
 dyn_side_hair_l_05 [{
   shapeType=Capsule
   length=4
   radius=0.91
   shapeName=dyn_side_hair_l_05_capsule
   center=<3.0711 -1.3786 0.5863>
   rotation=<rot 0 54.8351 -75.2566>
  }]
 dyn_side_hair_r_01 [{shapeType=Box, extentX=5.8125, extentY=4.1875, extentZ=1.625, center=<2.6057 0.0001 0>}]
 dyn_side_hair_r_02 [{shapeType=Box, extentX=5, extentY=2.5625, extentZ=1.625, center=<2.6057 0.0001 0>}]
 dyn_side_hair_r_03 [{shapeType=Box, extentX=3.625, extentY=0.75, extentZ=0.8125, center=<1.4749 -0.5118 0.4474>, rotation=<rot 19.6433 -9.9728 -3.3011>}]
 dyn_side_hair_r_04 [{shapeType=Capsule, length=2.6, radius=1, shapeName=dyn_side_hair_r_04_capsule, center=<1.1585 1.2361 0.0513>, rotation=<rot 0 120 -90>}]
 head [
  {shapeType=Capsule, length=7.4688, radius=5.4375, center=<0.5445 3.4826 0>, rotation=<rot 90 0 0>}
  {shapeType=Capsule, length=10.6, radius=2, shapeName=head_capsule, center=<1.1726 1.3262 0>}
 ]
 lowerarm_l [{shapeType=Capsule, length=21, radius=5, shapeName=lowerarm_l_capsule, center=<11.5 0 0>, rotation=<rot 90 0 0>}]
 lowerarm_r [{shapeType=Capsule, length=21, radius=5, shapeName=lowerarm_r_capsule, center=<-11.5137 0 0>, rotation=<rot 90 0 0>}]
 neck_01 [{shapeType=Capsule, length=6.6875, radius=4.5, center=<2.6233 0.663 0.0001>, rotation=<rot 90 0 0>}]
 pelvis [{shapeType=Capsule, length=8.1562, radius=12.1562, center=<-2.3989 -0.4861 0>}]
 spine_01 [{shapeType=Capsule, length=7.7188, radius=6.8438, center=<2.6802 1.5656 0>}]
 spine_03 [{shapeType=Capsule, length=2.7188, radius=9.4062, center=<1.3445 3.246 0>}]
 spine_04 [{shapeType=Capsule, length=5.125, radius=10.0625, center=<1.3926 2.539 0>}]
 spine_05 [
  {shapeType=Capsule, length=8.25, radius=8.9375, center=<1.3926 0.7102 0>}
  {shapeType=Capsule, length=7.625, radius=8, center=<9.4987 -1.2449 0>}
  {shapeType=Capsule, length=5.8125, radius=6.8438, center=<6.9456 4.288 0>}
 ]
 upperarm_l [{shapeType=Capsule, length=21.75, radius=5, shapeName=upperarm_l_capsule, center=<12.2749 0 0>, rotation=<rot 90 0 0>}]
 upperarm_r [{shapeType=Capsule, length=21.75, radius=4.2, shapeName=upperarm_r_capsule, center=<-12.5856 -1.2847 -0.156>, rotation=<rot 90 0 0>}]
}
constraints [
 {bone1Name=dyn_side_hair_l_01, bone2Name=head, swing1LimitDegrees=15, swing2LimitDegrees=10, twistLimitDegrees=5}
 {bone1Name=dyn_side_hair_l_02, bone2Name=dyn_side_hair_l_01, swing1LimitDegrees=10, swing2LimitDegrees=20, twistLimitDegrees=5}
 {bone1Name=dyn_side_hair_l_03, bone2Name=dyn_side_hair_l_02, swing1LimitDegrees=20, swing2LimitDegrees=30, twistLimitDegrees=5}
 {bone1Name=dyn_side_hair_r_01, bone2Name=head, swing1LimitDegrees=20, swing2LimitDegrees=15, twistLimitDegrees=5}
 {bone1Name=dyn_side_hair_r_02, bone2Name=dyn_side_hair_r_01, swing1LimitDegrees=20, swing2LimitDegrees=20, twistLimitDegrees=5}
 {bone1Name=dyn_side_hair_r_03, bone2Name=dyn_side_hair_r_02, swing1LimitDegrees=20, swing2LimitDegrees=20, twistLimitDegrees=5}
 {bone1Name=dyn_ponytail_MSTR, bone2Name=head, swing1LimitDegrees=1, swing2LimitDegrees=1, twistLimitDegrees=1}
 {bone1Name=dyn_ponytail_01, bone2Name=dyn_ponytail_00, swing1LimitDegrees=1, swing2LimitDegrees=1, twistLimitDegrees=1}
 {bone1Name=dyn_ponytail_02, bone2Name=dyn_ponytail_01, swing1LimitDegrees=40, swing2LimitDegrees=25, twistLimitDegrees=5}
 {bone1Name=dyn_ponytail_03, bone2Name=dyn_ponytail_02, swing1LimitDegrees=30, swing2LimitDegrees=25, twistLimitDegrees=5}
 {bone1Name=dyn_ponytail_04, bone2Name=dyn_ponytail_03, swing1LimitDegrees=30, swing2LimitDegrees=25, twistLimitDegrees=5}
 {bone1Name=dyn_ponytail_05, bone2Name=dyn_ponytail_04, swing1LimitDegrees=30, swing2LimitDegrees=25, twistLimitDegrees=5}
 {bone1Name=dyn_ponytail_06, bone2Name=dyn_ponytail_05, swing1LimitDegrees=30, swing2LimitDegrees=30, twistLimitDegrees=5}
 {
  bone1Name=head
  bone2Name=neck_01
  swing1LimitDegrees=45
  swing1Motion=Free
  swing2LimitDegrees=45
  swing2Motion=Free
  twistLimitDegrees=45
  twistMotion=Free
 }
 {
  bone1Name=clavicle_l
  bone2Name=spine_05
  swing1LimitDegrees=45
  swing1Motion=Free
  swing2LimitDegrees=45
  swing2Motion=Free
  twistLimitDegrees=45
  twistMotion=Free
 }
 {
  bone1Name=clavicle_r
  bone2Name=spine_05
  swing1LimitDegrees=45
  swing1Motion=Free
  swing2LimitDegrees=45
  swing2Motion=Free
  twistLimitDegrees=45
  twistMotion=Free
 }
 {
  bone1Name=neck_01
  bone2Name=spine_05
  swing1LimitDegrees=45
  swing1Motion=Free
  swing2LimitDegrees=45
  swing2Motion=Free
  twistLimitDegrees=45
  twistMotion=Free
 }
 {
  bone1Name=spine_05
  bone2Name=spine_04
  swing1LimitDegrees=45
  swing1Motion=Free
  swing2LimitDegrees=45
  swing2Motion=Free
  twistLimitDegrees=45
  twistMotion=Free
 }
 {
  bone1Name=spine_04
  bone2Name=spine_03
  swing1LimitDegrees=45
  swing1Motion=Free
  swing2LimitDegrees=45
  swing2Motion=Free
  twistLimitDegrees=45
  twistMotion=Free
 }
 {
  bone1Name=spine_03
  bone2Name=spine_01
  swing1LimitDegrees=45
  swing1Motion=Free
  swing2LimitDegrees=45
  swing2Motion=Free
  twistLimitDegrees=45
  twistMotion=Free
 }
 {
  bone1Name=spine_01
  bone2Name=pelvis
  swing1LimitDegrees=45
  swing1Motion=Free
  swing2LimitDegrees=45
  swing2Motion=Free
  twistLimitDegrees=45
  twistMotion=Free
 }
 {
  bone1Name=upperarm_l
  bone2Name=clavicle_l
  swing1LimitDegrees=45
  swing1Motion=Free
  swing2LimitDegrees=45
  swing2Motion=Free
  twistLimitDegrees=45
  twistMotion=Free
 }
 {
  bone1Name=upperarm_r
  bone2Name=clavicle_r
  swing1LimitDegrees=45
  swing1Motion=Free
  swing2LimitDegrees=45
  swing2Motion=Free
  twistLimitDegrees=45
  twistMotion=Free
 }
 {
  bone1Name=dyn_hat_bill
  bone2Name=head
  swing1LimitDegrees=45
  swing1Motion=Free
  swing2LimitDegrees=45
  swing2Motion=Free
  twistLimitDegrees=45
  twistMotion=Free
 }
 {bone1Name=dyn_ponytail_00, bone2Name=dyn_ponytail_MSTR, swing1LimitDegrees=1, swing2LimitDegrees=1, twistLimitDegrees=1}
 {bone1Name=dyn_side_hair_l_04, bone2Name=dyn_side_hair_l_01, swing1LimitDegrees=15, swing2LimitDegrees=5, twistLimitDegrees=5}
 {bone1Name=dyn_side_hair_l_05, bone2Name=dyn_side_hair_l_01, swing1LimitDegrees=30, swing2LimitDegrees=5, twistLimitDegrees=5}
 {
  bone1Name=lowerarm_r
  bone2Name=upperarm_r
  swing1LimitDegrees=45
  swing1Motion=Free
  swing2LimitDegrees=45
  swing2Motion=Free
  twistLimitDegrees=45
  twistMotion=Free
 }
 {
  bone1Name=lowerarm_l
  bone2Name=upperarm_l
  swing1LimitDegrees=45
  swing1Motion=Free
  swing2LimitDegrees=45
  swing2Motion=Free
  twistLimitDegrees=45
  twistMotion=Free
 }
 {bone1Name=dyn_side_hair_r_04, bone2Name=dyn_side_hair_r_01, swing1LimitDegrees=20, swing2LimitDegrees=5, twistLimitDegrees=5}
]
