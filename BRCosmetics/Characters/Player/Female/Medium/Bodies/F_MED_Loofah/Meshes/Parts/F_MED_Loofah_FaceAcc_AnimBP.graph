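# Anim Blueprint for the Loofah character's face-accessory part, animating hat and ponytail over a pose copied from the parent and body meshes.
# Normally a Control Rig feeds RigidBody physics with state-driven gravity.
# With backbling, a second physics asset, spine/hat constraints and bone offsets shape the ponytail, adjusted further in non-emote front end.
# Event graph derives head-angle alphas for corrective offsets; disabled rigid-body simulation falls back to fixed offsets.
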AnimBlueprint /BRCosmetics/Characters/Player/Female/Medium/Bodies/F_MED_Loofah/Meshes/Parts/F_MED_Loofah_FaceAcc_AnimBP
target_skeleton=&/F_MED_Loofah_FaceAcc_Skeleton
graphs [
 AnimGraph {nodes [
   "Constraint - dyn_ponytail_01 (PP)"!(ComponentPose, Alpha=1, ConstraintWeights_0=0.5, ConstraintWeights_1:float=0.3) -> (Pose) {
    alphaBoolBlend {blendOption=Linear}
    boneToModify {boneName=dyn_ponytail_01}
    constraintSetup [{targetBone {boneName=hood}}, {targetBone {boneName=dyn_hat_bill}}]
    constraintWeights [1,1]
   }
   n1 = "Transform (Modify) Bone - Bone: dyn_ponytail_01"(Alpha=Get head delta() -> (head delta:double), Translation="2,0,2", Rotation="-25,0,0", Scale="1,1,1") {
    rotationMode=BMM_Additive
    rotationSpace=BCS_BoneSpace
    scaleMode=BMM_Additive
    scaleSpace=BCS_BoneSpace
    translationMode=BMM_Additive
    translationSpace=BCS_BoneSpace
    alphaBoolBlend {blendOption=Linear}
    boneToModify {boneName=dyn_ponytail_01}
    rotation {pitch=-25}
    translation {x=2, z=2}
   }
   n2 = "Transform (Modify) Bone - Bone: dyn_ponytail_01"(ComponentPose=@n1, Alpha=Get HeadLeftAlpha() -> (HeadLeftAlpha:float), Translation="1,0,2", Rotation="0,0,0", Scale="1,1,1") {
    rotationMode=BMM_Additive
    rotationSpace=BCS_BoneSpace
    scaleMode=BMM_Additive
    scaleSpace=BCS_BoneSpace
    translationMode=BMM_Additive
    translationSpace=BCS_BoneSpace
    alphaBoolBlend {blendOption=Linear}
    boneToModify {boneName=dyn_ponytail_01}
    translation {x=1, z=2}
   }
   "Transform (Modify) Bone - Bone: dyn_ponytail_01"(ComponentPose=@n2, Alpha=Get HeadUp() -> (HeadUp:double), Translation="3,0,3", Rotation="-30,0,0", Scale="1,1,1") {
    rotationMode=BMM_Additive
    rotationSpace=BCS_BoneSpace
    scaleMode=BMM_Additive
    scaleSpace=BCS_BoneSpace
    translationMode=BMM_Additive
    translationSpace=BCS_BoneSpace
    alphaBoolBlend {blendOption=Linear}
    boneToModify {boneName=dyn_ponytail_01}
    rotation {pitch=-30}
    translation {x=3, z=3}
   }
   "Transform (Modify) Bone - Bone: dyn_ponytail_MSTR"(Alpha=Get head delta(), Translation="0,0,2", Rotation="0,0,0", Scale="1,1,1") {
    rotationMode=BMM_Additive
    rotationSpace=BCS_BoneSpace
    scaleMode=BMM_Additive
    scaleSpace=BCS_BoneSpace
    translationMode=BMM_Additive
    translationSpace=BCS_BoneSpace
    alphaBoolBlend {blendOption=Linear}
    boneToModify {boneName=dyn_ponytail_MSTR}
    translation {z=2}
   }
   n3 = "Transform (Modify) Bone - Bone: dyn_ponytail_03"(
    ComponentPose=Local To Component(LocalPose=Use cached pose 'FrontEndCache'())
    Alpha=1
    Translation="0,0,0"
    Rotation="-10,10,0"
    Scale="1,1,1"
   ) {
    alpha=0
    rotationMode=BMM_Additive
    rotationSpace=BCS_BoneSpace
    translationMode=BMM_Additive
    translationSpace=BCS_BoneSpace
    alphaBoolBlend {blendOption=Linear}
    boneToModify {boneName=dyn_ponytail_03}
    rotation {pitch=-10, yaw=10}
   }
   n4 = "Transform (Modify) Bone - Bone: dyn_ponytail_04"(ComponentPose=@n3, Alpha=1, Translation="0,0,0", Rotation="-7,0,0", Scale="1,1,1") {
    alpha=0
    rotationMode=BMM_Additive
    rotationSpace=BCS_BoneSpace
    translationMode=BMM_Additive
    translationSpace=BCS_BoneSpace
    alphaBoolBlend {blendOption=Linear}
    boneToModify {boneName=dyn_ponytail_04}
    rotation {pitch=-7}
   }
   n5 = "Transform (Modify) Bone - Bone: dyn_ponytail_05"(ComponentPose=@n4, Alpha=1, Translation="-1.5,1.5,5", Rotation="10,20,0", Scale="1,1,1") {
    alpha=0
    rotationMode=BMM_Additive
    rotationSpace=BCS_BoneSpace
    translationMode=BMM_Additive
    translationSpace=BCS_BoneSpace
    alphaBoolBlend {blendOption=Linear}
    boneToModify {boneName=dyn_ponytail_05}
    rotation {pitch=10, yaw=20}
    translation=<-1.5 1.5 5>
   }
   n6 = Get bIsPlayingEmote()
   n7 = Blend Poses by bool(
    BlendPose_0=Component To Local(ComponentPose=@n5)
    BlendPose_1=Use cached pose 'FrontEndCache'()
    BlendTime_0=0.5
    BlendTime_1=0.5
    bActiveValue=AND Boolean(A=Get bIsInFrontEnd(), B=NOT Boolean(A=@n6))
   ) {blendType=Linear}
   n8 = Control Rig(Source=Use cached pose 'baseCache'()) {
    alphaBoolBlend {blendOption=Linear}
    controlRigAssetReference {blueprintRigClass=&/F_MED_Loofah_FaceAcc_ControlRig.F_MED_Loofah_FaceAcc_ControlRig_C}
   }
   n9 = RigidBody!(ComponentPose=Local To Component(LocalPose=@n8), Alpha, OverrideWorldGravity:Vector=Get GravityOverride(), ExternalForce) -> (Pose) {
    bClampLinearTranslationLimitToRefPose=true
    bOverrideWorldGravity=true
    overridePhysicsAsset=&/F_MED_Loofah_FaceAcc_Dyn
    simulationSpace=BaseBoneSpace
    alphaBoolBlend {blendOption=Linear}
    baseBoneRef {boneName=spine_05}
    componentAppliedLinearAccClamp=<3000 3000 3000>
    componentLinearAccScale=<0.45 0.3 0.3>
    componentLinearVelScale=<0.45 0.3 0.3>
    overrideWorldGravity {z=-980}
   }
   n10 = Component To Local(ComponentPose=@n9)
   n11 = Get isWearingBackbling()
   n12 = "Transform (Modify) Bone - Bone: dyn_ponytail_CV3"(ComponentPose=Local To Component(LocalPose=Use cached pose 'baseCache'()), Alpha=1, Translation="0,0,-3", Rotation="0,0,0", Scale="1,1,1") {
    rotationMode=BMM_Additive
    rotationSpace=BCS_BoneSpace
    scaleMode=BMM_Additive
    scaleSpace=BCS_BoneSpace
    translationMode=BMM_Additive
    translationSpace=BCS_BoneSpace
    alphaBoolBlend {blendOption=Linear}
    alphaScaleBiasClamp {bInterpResult=true, interpSpeedDecreasing=20, interpSpeedIncreasing=20}
    boneToModify {boneName=dyn_ponytail_CV3}
    translation {z=-3}
   }
   n13 = "Transform (Modify) Bone - Bone: dyn_ponytail_CV4"(ComponentPose=@n12, Alpha=1, Translation="0,0,-5", Rotation="0,0,0", Scale="1,1,1") {
    rotationMode=BMM_Additive
    rotationSpace=BCS_BoneSpace
    scaleMode=BMM_Additive
    scaleSpace=BCS_BoneSpace
    translationMode=BMM_Additive
    translationSpace=BCS_BoneSpace
    alphaBoolBlend {blendOption=Linear}
    alphaScaleBiasClamp {bInterpResult=true, interpSpeedDecreasing=20, interpSpeedIncreasing=20}
    boneToModify {boneName=dyn_ponytail_CV4}
    translation {z=-5}
   }
   n14 = "Transform (Modify) Bone - Bone: dyn_ponytail_CV5"(ComponentPose=@n13, Alpha=1, Translation="0,0,-10", Rotation="0,0,0", Scale="1,1,1") {
    rotationMode=BMM_Additive
    rotationSpace=BCS_BoneSpace
    scaleMode=BMM_Additive
    scaleSpace=BCS_BoneSpace
    translationMode=BMM_Additive
    translationSpace=BCS_BoneSpace
    alphaBoolBlend {blendOption=Linear}
    alphaScaleBiasClamp {bInterpResult=true, interpSpeedDecreasing=20, interpSpeedIncreasing=20}
    boneToModify {boneName=dyn_ponytail_CV5}
    translation {z=-10}
   }
   Output Pose(Result=Blend Poses by bool(
     BlendPose_0=Component To Local(ComponentPose=@n14)
     BlendPose_1=Blend Poses by bool(BlendPose_0=@n7, BlendPose_1=@n10, bActiveValue=@n11) {blendType=Linear}
     bActiveValue=Get RBAN off() -> (RBAN off:bool)
    ) {blendType=Linear}) {layerGroup=None}
   n15 = Get head delta()
   n16 = "Transform (Modify) Bone - Bone: dyn_hat"(
    ComponentPose=Local To Component(LocalPose=Use cached pose 'baseCache'())
    Alpha=@n15
    Translation="0.5,0,0"
    Rotation="0,9,0"
    Scale="1,1,1"
   ) {
    rotationMode=BMM_Additive
    rotationSpace=BCS_BoneSpace
    translationMode=BMM_Additive
    translationSpace=BCS_BoneSpace
    alphaBoolBlend {blendOption=Linear}
    alphaScaleBiasClamp {bInterpResult=true, interpSpeedDecreasing=20, interpSpeedIncreasing=20}
    boneToModify {boneName=dyn_hat}
    rotation {yaw=9}
    translation {x=0.5}
   }
   n17 = "Transform (Modify) Bone - Bone: dyn_hat_bill"(ComponentPose=@n16, Alpha=@n15, Translation="0,0,0", Rotation="6,25,0", Scale="1,1,1") {
    rotationMode=BMM_Additive
    rotationSpace=BCS_BoneSpace
    translationMode=BMM_Additive
    translationSpace=BCS_BoneSpace
    alphaBoolBlend {blendOption=Linear}
    alphaScaleBiasClamp {bInterpResult=true, interpSpeedDecreasing=20, interpSpeedIncreasing=20}
    boneToModify {boneName=dyn_hat_bill}
    rotation {pitch=6, yaw=25}
   }
   n18 = RigidBody(ComponentPose=@n17, OverrideWorldGravity=Get GravityOverride()) {
    bClampLinearTranslationLimitToRefPose=true
    bOverrideWorldGravity=true
    overridePhysicsAsset=&/F_MED_Loofah_FaceAcc_Dyn_w_backbling
    simulationSpace=BaseBoneSpace
    alphaBoolBlend {blendOption=Linear}
    baseBoneRef {boneName=spine_05}
    componentAppliedLinearAccClamp=<3000 3000 3000>
    componentLinearAccScale=<0.45 0.3 0.3>
    componentLinearVelScale=<0.45 0.3 0.3>
    overrideWorldGravity {z=-980}
   }
   n19 = "Constraint - dyn_ponytail_00 (PP)"!(ComponentPose=@n18, Alpha=1, ConstraintWeights_0=0.5, ConstraintWeights_1:float=0.5) -> (Pose) {
    alphaBoolBlend {blendOption=Linear}
    boneToModify {boneName=dyn_ponytail_00}
    constraintSetup [{targetBone {boneName=dyn_hat_bill}}, {targetBone {boneName=dyn_hat}}]
    constraintWeights [1,1]
   }
   n20 = "Constraint - dyn_ponytail_01 (PP)"(ComponentPose=@n19, Alpha=1, ConstraintWeights_0=0.8, ConstraintWeights_1=0.2) {
    alphaBoolBlend {blendOption=Linear}
    boneToModify {boneName=dyn_ponytail_01}
    constraintSetup [{targetBone {boneName=dyn_hat_bill}}, {targetBone {boneName=spine_05}}]
    constraintWeights [1,1]
   }
   n21 = "Constraint - dyn_ponytail_02 (PPP)"!(ComponentPose=@n20, Alpha=1, ConstraintWeights_0=0.3, ConstraintWeights_1:float=0.4, ConstraintWeights_2:float=0.3) -> (Pose) {
    alphaBoolBlend {blendOption=Linear}
    boneToModify {boneName=dyn_ponytail_02}
    constraintSetup [{targetBone {boneName=dyn_hat_bill}}, {targetBone {boneName=spine_05}}, {targetBone {boneName=dyn_ponytail_03}}]
    constraintWeights [1,1,1]
   }
   n22 = "Constraint - dyn_ponytail_03 (PP)"!(ComponentPose=@n21, Alpha=1, ConstraintWeights_0=0.5, ConstraintWeights_1:float=0.5) -> (Pose) {
    alphaBoolBlend {blendOption=Linear}
    boneToModify {boneName=dyn_ponytail_03}
    constraintSetup [{targetBone {boneName=neck_02}}, {targetBone {boneName=spine_05}}]
    constraintWeights [1,1]
   }
   n23 = "Constraint - dyn_ponytail_04 (PP)"!(ComponentPose=@n22, Alpha=1, ConstraintWeights_0=0.5, ConstraintWeights_1:float=0.5) -> (Pose) {
    alphaBoolBlend {blendOption=Linear}
    boneToModify {boneName=dyn_ponytail_04}
    constraintSetup [{targetBone {boneName=spine_05}}, {targetBone {boneName=spine_04}}]
    constraintWeights [1,1]
   }
   n24 = "Constraint - dyn_ponytail_05 (PP)"!(ComponentPose=@n23, Alpha=1, ConstraintWeights_0=0.5, ConstraintWeights_1:float=0.5) -> (Pose) {
    alphaBoolBlend {blendOption=Linear}
    boneToModify {boneName=dyn_ponytail_05}
    constraintSetup [{targetBone {boneName=spine_04}}, {targetBone {boneName=spine_03}}]
    constraintWeights [1,1]
   }
   n25 = "Constraint - dyn_ponytail_06 (PP)"!(ComponentPose=@n24, Alpha=1, ConstraintWeights_0=0.5, ConstraintWeights_1:float=0.5) -> (Pose) {
    alphaBoolBlend {blendOption=Linear}
    boneToModify {boneName=dyn_ponytail_06}
    constraintSetup [{targetBone {boneName=spine_03}}, {targetBone {boneName=spine_02}}]
    constraintWeights [1,1]
   }
   n26 = "Transform (Modify) Bone - Bone: dyn_ponytail_01"(ComponentPose=@n25, Alpha=1, Translation="0,0,0", Rotation="0,0,0", Scale="1,1,1") {
    rotationMode=BMM_Additive
    rotationSpace=BCS_BoneSpace
    scaleMode=BMM_Additive
    scaleSpace=BCS_BoneSpace
    translationMode=BMM_Additive
    translationSpace=BCS_BoneSpace
    alphaBoolBlend {blendOption=Linear}
    alphaScaleBiasClamp {bInterpResult=true, interpSpeedDecreasing=20, interpSpeedIncreasing=20}
    boneToModify {boneName=dyn_ponytail_01}
   }
   n27 = Get HeadUp()
   n28 = "Transform (Modify) Bone - Bone: dyn_hat_bill"(ComponentPose=@n26, Alpha=@n27, Translation="0,0,0", Rotation="0,10,0", Scale="1,1,1") {
    rotationMode=BMM_Additive
    rotationSpace=BCS_BoneSpace
    scaleMode=BMM_Additive
    scaleSpace=BCS_BoneSpace
    translationMode=BMM_Additive
    translationSpace=BCS_BoneSpace
    alphaBoolBlend {blendOption=Linear}
    boneToModify {boneName=dyn_hat_bill}
    rotation {yaw=10}
   }
   n29 = "Transform (Modify) Bone - Bone: dyn_hat"(ComponentPose=@n28, Alpha=@n27, Translation="1,0,0", Rotation="0,3,0", Scale="1,1,1") {
    rotationMode=BMM_Additive
    rotationSpace=BCS_BoneSpace
    scaleMode=BMM_Additive
    scaleSpace=BCS_BoneSpace
    translationMode=BMM_Additive
    translationSpace=BCS_BoneSpace
    alphaBoolBlend {blendOption=Linear}
    boneToModify {boneName=dyn_hat}
    rotation {yaw=3}
    translation {x=1}
   }
   n30 = "Transform (Modify) Bone - Bone: dyn_ponytail_02"(ComponentPose=@n29, Alpha=1, Translation="0,0,0", Rotation="0,0,0", Scale="1,1,1") {
    rotationMode=BMM_Additive
    rotationSpace=BCS_BoneSpace
    scaleMode=BMM_Additive
    scaleSpace=BCS_BoneSpace
    translationMode=BMM_Additive
    translationSpace=BCS_BoneSpace
    alphaBoolBlend {blendOption=Linear}
    alphaScaleBiasClamp {bInterpResult=true, interpSpeedDecreasing=20, interpSpeedIncreasing=20}
    boneToModify {boneName=dyn_ponytail_02}
   }
   n31 = "Transform (Modify) Bone - Bone: dyn_ponytail_03"(ComponentPose=@n30, Alpha=1, Translation="0,0,-3", Rotation="0,0,0", Scale="1,1,0.8") {
    rotationMode=BMM_Additive
    rotationSpace=BCS_BoneSpace
    scaleMode=BMM_Additive
    scaleSpace=BCS_BoneSpace
    translationMode=BMM_Additive
    translationSpace=BCS_BoneSpace
    alphaBoolBlend {blendOption=Linear}
    alphaScaleBiasClamp {bInterpResult=true, interpSpeedDecreasing=20, interpSpeedIncreasing=20}
    boneToModify {boneName=dyn_ponytail_03}
    scale {z=0.8}
    translation {z=-3}
   }
   n32 = "Transform (Modify) Bone - Bone: dyn_ponytail_04"(ComponentPose=@n31, Alpha=1, Translation="0,0,-5", Rotation="-10,0,0", Scale="1,1,0.8") {
    rotationMode=BMM_Additive
    rotationSpace=BCS_BoneSpace
    scaleMode=BMM_Additive
    scaleSpace=BCS_BoneSpace
    translationMode=BMM_Additive
    translationSpace=BCS_BoneSpace
    alphaBoolBlend {blendOption=Linear}
    alphaScaleBiasClamp {bInterpResult=true, interpSpeedDecreasing=20, interpSpeedIncreasing=20}
    boneToModify {boneName=dyn_ponytail_04}
    rotation {pitch=-10}
    scale {z=0.8}
    translation {z=-5}
   }
   n33 = "Transform (Modify) Bone - Bone: dyn_ponytail_05"(ComponentPose=@n32, Alpha=1, Translation="0,0,-5", Rotation="0,0,0", Scale="1,1,1") {
    rotationMode=BMM_Additive
    rotationSpace=BCS_BoneSpace
    scaleMode=BMM_Additive
    scaleSpace=BCS_BoneSpace
    translationMode=BMM_Additive
    translationSpace=BCS_BoneSpace
    alphaBoolBlend {blendOption=Linear}
    alphaScaleBiasClamp {bInterpResult=true, interpSpeedDecreasing=20, interpSpeedIncreasing=20}
    boneToModify {boneName=dyn_ponytail_05}
    translation {z=-5}
   }
   n34 = "Transform (Modify) Bone - Bone: dyn_ponytail_06"(ComponentPose=@n33, Alpha=1, Translation="0,0,0", Rotation="20,0,0", Scale="1,1,1") {
    rotationMode=BMM_Additive
    rotationSpace=BCS_BoneSpace
    scaleMode=BMM_Additive
    scaleSpace=BCS_BoneSpace
    translationMode=BMM_Additive
    translationSpace=BCS_BoneSpace
    alphaBoolBlend {blendOption=Linear}
    alphaScaleBiasClamp {bInterpResult=true, interpSpeedDecreasing=20, interpSpeedIncreasing=20}
    boneToModify {boneName=dyn_ponytail_06}
    rotation {pitch=20}
   }
   Save cached pose 'FrontEndCache'(Pose=Component To Local(ComponentPose=@n34)) {cacheName=FrontEndCache}
   n35 = Layered blend per bone(
    BasePose=Copy Pose From Mesh() {bUseAttachedParent=true}
    BlendPoses_0=Copy Pose From Mesh(SourceMeshComponent=Get body_mesh() -> (body_mesh:Skeletal Mesh Component Object Reference))
   ) {layerSetup [{branchFilters [{boneName=clavicle_l}, {boneName=dyn_horn_l_01}, {boneName=clavicle_r}, {boneName=dyn_horn_r_01}]}]}
   Save cached pose 'baseCache'(Pose=@n35) {cacheName=baseCache}
  ]}
 EventGraph {nodes [
   Event BlueprintUpdateAnimation()
   > n1 = Sequence!(execute) -> (then_0, then_1, then_2:Exec, then_3:Exec, then_4:Exec, then_5:Exec)
   n2 = CheckForBackbling()
   Set isWearingBackbling(execute=[@n7.then, @n8.then, @n1.then_0], isWearingBackbling=OR Boolean(A=@n2, B=@n2.bIsBackblingLong))
   > Set Skel_Mesh(Skel_Mesh=GetOwnerSkeletalMeshComponent())
   Set head delta(
    execute=@n1.then_1
    head delta:double=GetAlphaFromReferenceTransformDelta(
     SkeletalMeshComponent=Get Skel_Mesh()
     BoneName=head
     BaseName=spine_05
     DeltaComponentType=QuaternionTwist
     InputRange=(X=18,Y=59)
     TwistAxis="0,0,1"
    )
   ) -> (Output_Get:double)
   > Set HeadLeftAlpha(HeadLeftAlpha:float=GetAlphaFromReferenceTransformDelta(
     SkeletalMeshComponent=Get Skel_Mesh()
     BoneName=head
     BaseName=spine_01
     DeltaComponentType=EulerX
     InputRange=(X=0,Y=-15)
     TwistAxis="0,0,1"
    )) -> (Output_Get:float)
   > Set HeadUpAlpha(HeadUpAlpha:float=GetAlphaFromReferenceTransformDelta(SkeletalMeshComponent=Get Skel_Mesh(), BoneName=head, BaseName=neck_01, InputRange=(X=0,Y=-30), TwistAxis="0,0,1")) -> (Output_Get:float)
   > Set HeadUp(HeadUp:double=GetAlphaFromReferenceTransformDelta(SkeletalMeshComponent=Get Skel_Mesh(), BoneName=head, BaseName=neck_02, InputRange=(X=-40,Y=-50), TwistAxis="0,0,1")) -> (Output_Get:double)
   > n3 = Gravity Override State Function DEPRECATED(
    CharacterPartAnimInstance=Self-Reference()
    GameTimeInSeconds=GetGameTimeInSeconds()
    GravityOverrideParams=Get Gravity Override Params()
    RBAN Params=Get RBAN Params()
   )
   > Set GravityOverride(GravityOverride=@n3)
   n4 = BreakTransform(InTransform=GetDeltaTransformFromRefPose(self=GetOwnerSkeletalMeshComponent(), BoneName=spine_05, BaseName=spine_01))
   Set hat bill Rotate(execute=@n1.then_2, hat bill Rotate:double=MapRangeClamped(Value=BreakRotator(InRot=@n4.Rotation).Yaw, InRangeA=0, InRangeB=-20)) -> (Output_Get:double)
   n5 = BreakTransform(InTransform=GetDeltaTransformFromRefPose(self=GetOwnerSkeletalMeshComponent(), BoneName=head, BaseName=spine_05))
   Set Hat Rotate(execute=@n1.then_3, Hat Rotate:double=MapRangeClamped(Value=BreakRotator(InRot=@n5.Rotation).Yaw, InRangeA=-60, InRangeB=-65)) -> (Output_Get:double)
   n6 = Cast To PlayerPawn_Athena(execute=@n1.then_5, Object=TryGetPawnOwner())
   > Set body_mesh(body_mesh:Skeletal Mesh Component Object Reference=GetSkeletalMeshForPartType(self=@n6)) -> (Output_Get:Skeletal Mesh Component Object Reference)
   Set RBAN off(execute=@n1.then_4, RBAN off:bool=NOT Boolean(A=Get bRigidBodyEnabled())) -> (Output_Get:bool)
   n7 = Event OnCharacterCustomizationLoaded()
   n8 = Event BlueprintInitializeAnimation()
   TryGetPawnOwner()
   Do Once()
  ]}
]
variables [
 body_mesh {type=SkeletalMeshComponent}
 Gravity Override Params {type=GravityOverrideParamsStruct}
 GravityOverride {type=Vector}
 hat bill Rotate {type=number}
 Hat Rotate {type=number}
 head delta {type=number}
 HeadLeftAlpha {type=number, default=1}
 HeadUp {type=number}
 HeadUpAlpha {type=number}
 isWearingBackbling {type=boolean}
 RBAN off {type=boolean}
 RBAN Params {
  type=RBANParamsStruct
  default {
   dBNO {simSpaceSettings {
     maxAngularAcceleration=1000
     maxAngularVelocity=3
     maxLinearAcceleration=300
     maxLinearVelocity=700
     velocityScaleZ=0
     externalLinearDragV=<0 0.5 0>
    }}
   dBNO_Carried {simSpaceSettings {
     maxAngularAcceleration=1000
     maxAngularVelocity=3
     maxLinearAcceleration=300
     maxLinearVelocity=700
     velocityScaleZ=0
     externalLinearDragV=<0 0.5 0>
    }}
   emoteOrMelee {simSpaceSettings {
     maxAngularAcceleration=1000
     maxAngularVelocity=3
     maxLinearAcceleration=300
     maxLinearVelocity=700
     velocityScaleZ=0
     externalLinearDragV=<0 0.5 0>
    }}
   front_End {simSpaceSettings {
     maxAngularAcceleration=1000
     maxAngularVelocity=3
     maxLinearAcceleration=300
     maxLinearVelocity=700
     velocityScaleZ=0
     externalLinearDragV=<0 0.5 0>
    }}
   onGround_Crouch_Moving {simSpaceSettings {
     maxAngularAcceleration=1000
     maxAngularVelocity=3
     maxLinearAcceleration=300
     maxLinearVelocity=700
     velocityScaleZ=0
     externalLinearDragV=<0 0.5 0>
    }}
   onGround_Falling {simSpaceSettings {
     maxAngularAcceleration=1000
     maxAngularVelocity=3
     maxLinearAcceleration=300
     maxLinearVelocity=700
     velocityScaleZ=0
     externalLinearDragV=<0 0.5 0>
    }}
   onGround_Moving {simSpaceSettings {
     maxAngularAcceleration=1000
     maxAngularVelocity=3
     maxLinearAcceleration=300
     maxLinearVelocity=700
     velocityScaleZ=0
     externalLinearDragV=<0 0.5 0>
    }}
   onGround_Standing {simSpaceSettings {
     maxAngularAcceleration=1000
     maxAngularVelocity=3
     maxLinearAcceleration=300
     maxLinearVelocity=700
     velocityScaleZ=0
     externalLinearDragV=<0 0.5 0>
    }}
   rail_Grinding {simSpaceSettings {
     maxAngularAcceleration=1000
     maxAngularVelocity=3
     maxLinearAcceleration=300
     maxLinearVelocity=700
     velocityScaleZ=0
     worldAlpha=0.1
     externalLinearDragV=<0 0.5 0>
    }}
   riding_Moving {simSpaceSettings {
     maxAngularAcceleration=1000
     maxAngularVelocity=3
     maxLinearAcceleration=300
     maxLinearVelocity=700
     velocityScaleZ=0
     externalLinearDragV=<0 0.5 0>
    }}
   riding_Moving_Targeting {simSpaceSettings {
     maxAngularAcceleration=1000
     maxAngularVelocity=3
     maxLinearAcceleration=300
     maxLinearVelocity=700
     velocityScaleZ=0
     externalLinearDragV=<0 0.5 0>
    }}
   riding_Standing {simSpaceSettings {
     maxAngularAcceleration=1000
     maxAngularVelocity=3
     maxLinearAcceleration=300
     maxLinearVelocity=700
     velocityScaleZ=0
     externalLinearDragV=<0 0.5 0>
    }}
   skydive_DivingDown {simSpaceSettings {
     maxAngularAcceleration=1000
     maxAngularVelocity=3
     maxLinearAcceleration=300
     maxLinearVelocity=700
     velocityScaleZ=0
     externalLinearDragV=<0 0.5 0>
    }}
   skydive_FloatingDown {
    componentLinearAccScale=<0 0.5 0>
    componentLinearVelScale=<0 0.5 0>
    simSpaceSettings {
     maxAngularAcceleration=1000
     maxAngularVelocity=3
     maxLinearAcceleration=300
     maxLinearVelocity=700
     velocityScaleZ=0
     externalLinearDragV=<0 0.5 0>
    }
   }
   skydive_GoingUp {simSpaceSettings {
     maxAngularAcceleration=1000
     maxAngularVelocity=3
     maxLinearAcceleration=300
     maxLinearVelocity=700
     velocityScaleZ=0
     externalLinearDragV=<0 0.5 0>
    }}
   skydive_Parachuting {
    componentLinearAccScale=<0 0.1 0.3>
    componentLinearVelScale=<0 0.1 0.3>
    simSpaceSettings {
     maxAngularAcceleration=1000
     maxAngularVelocity=3
     maxLinearAcceleration=300
     maxLinearVelocity=700
     velocityScaleZ=0
     externalLinearDragV=<0 0.5 0>
    }
   }
   skydive_Surfing {simSpaceSettings {
     maxAngularAcceleration=1000
     maxAngularVelocity=3
     maxLinearAcceleration=300
     maxLinearVelocity=700
     velocityScaleZ=0
     externalLinearDragV=<0 0.5 0>
    }}
   swimming {simSpaceSettings {
     maxAngularAcceleration=1000
     maxAngularVelocity=3
     maxLinearAcceleration=300
     maxLinearVelocity=700
     velocityScaleZ=0
     externalLinearDragV=<0 0.5 0>
    }}
   tactical_Sprint {simSpaceSettings {
     maxAngularAcceleration=1000
     maxAngularVelocity=3
     maxLinearAcceleration=300
     maxLinearVelocity=700
     velocityScaleZ=0
     externalLinearDragV=<0 0.5 0>
    }}
   ziplining {simSpaceSettings {
     maxAngularAcceleration=1000
     maxAngularVelocity=3
     maxLinearAcceleration=300
     maxLinearVelocity=700
     velocityScaleZ=0
     externalLinearDragV=<0 0.5 0>
    }}
  }
 }
 Skel_Mesh {type=SkeletalMeshComponent}
]
